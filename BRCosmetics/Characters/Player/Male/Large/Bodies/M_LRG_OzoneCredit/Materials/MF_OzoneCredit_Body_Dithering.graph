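# Material function for the OzoneCredit body that dither-fades regions selected by a DitherMask input, apparently during aiming.
# MPC_Ozone_Dithering supplies aim state and a start time, with Use_PacificSweater selecting the Sweater variants.
# Elapsed time drives a quick quadratic ease-out that, scaled by aim, blends the threshold toward the mask.
# DitherTemporalAA_Fortnite output is blended with the incoming OpacityMask by the aim-weighted mask.
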MaterialFunction /BRCosmetics/Characters/Player/Male/Large/Bodies/M_LRG_OzoneCredit/Materials/MF_OzoneCredit_Body_Dithering
graph {nodes [
  n1 = GetMaterialAttributes(None=FunctionInput() {inputName=MA, inputType=FunctionInput_MaterialAttributes}) {attributeGetTypes [#1]}
  n2 = FunctionInput() {bUsePreviewValueAsDefault=true, inputName=DitherMask}
  n3 = CollectionParameter() {collection=&/MPC_Ozone_Dithering, parameterName=BP_GetTime}
  n4 = CollectionParameter() {
   collection=&MPC_Ozone_Dithering
   parameterName=BP_GetTime_Sweater
  }
  n5 = Subtract(A=Time(), B=LinearInterpolate(A=@n3, B=@n4, Alpha=ScalarParameter() {group=Body_Dithering, parameterName=Use_PacificSweater}) {constAlpha=0})
  n6 = OneMinus(None=Saturate(None=Multiply(A=@n5) {constB=3.5}))
  n7 = OneMinus(None=Multiply(A=@n6, B=@n6))
  n8 = CollectionParameter() {collection=&MPC_Ozone_Dithering, parameterName=BP_Aim_Sweater}
  n9 = LinearInterpolate(
   A=CollectionParameter() {collection=&MPC_Ozone_Dithering, parameterName=BP_Aim}
   B=@n8
   Alpha=ScalarParameter() {group=Body_Dithering, parameterName=Use_PacificSweater}
  ) {constAlpha=0}
  n10 = MaterialFunctionCall(Alpha Threshold=LinearInterpolate(B=LinearInterpolate(B=@n2, Alpha=@n7) {constA=1}, Alpha=@n9) {constA=1}) {materialFunction=&DitherTemporalAA_Fortnite}
  n11 = LinearInterpolate(B=@n2, Alpha=@n9) {constA=1}
  FunctionOutput(None=SetMaterialAttributes!(MaterialAttributes=@n1.MaterialAttributes, OpacityMask=LinearInterpolate(A=@n10.Result, B=@n1.OpacityMask, Alpha=@n11), EmissiveColor) {attributeSetTypes [#1, #2]})
  ScalarParameter() {parameterName=BP_Aim}
  ScalarParameter() {defaultValue=9.999999680285692e+37, parameterName=BP_GetTime}
 ]}
