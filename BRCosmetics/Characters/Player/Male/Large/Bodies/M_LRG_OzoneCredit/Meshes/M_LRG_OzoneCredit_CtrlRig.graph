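# Control Rig for the large male OzoneCredit body cosmetic, applying procedural fixes to deform and accessory bones.
# Knee, ankle and elbow angles become alphas that squash foil scaler bones.
# Arm retargeting uses PBIK with elbow-to-head distance correction; fingers copy base bones.
# Spherical pose readers move pauldrons and flaps (Stick of Truth uses one right-arm driver), then interpolated gameplay flags adjust back flaps, neck seat and pauldrons for slides, emotes and parachuting.
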ControlRigBlueprint /BRCosmetics/Characters/Player/Male/Large/Bodies/M_LRG_OzoneCredit/Meshes/M_LRG_OzoneCredit_CtrlRig
graphs [
 CollapseNode_ContainedGraph {nodes [
   n1 = RigVMFunctionEntryNode()
   RigVMFunction_MathBoolToFloat = &RigVMFunction_MathBoolToFloat(Value=RigVMVariableNode() {variable=Pub_kneesliding}.Value)
   AlphaInterp = &RigVMFunction_AlphaInterp(InterpSpeedDecreasing=5, InterpSpeedIncreasing=5, Scale=1, bClampResult=false, Value=@RigVMFunction_MathBoolToFloat.Result)
   &RigUnit_ModifyTransforms(
    ItemToModify="((Item=(Type=Bone,Name=\"dyn_backFlap_1\"),Transform=(Rotation=(X=0,Y=0,Z=-0.4226,W=0.9063),Translation=(X=0,Y=0,Z=0),Scale3D=(X=1,Y=1,Z=1))),(Item=(Type=Bone,Name=\"dyn_backFlap_2\"),Transform=(Rotation=(X=0,Y=0,Z=-0.0872,W=0.9962),Translation=(X=0,Y=0,Z=0),S ...(truncated)"
    WeightMinimum=0
    Weight=@AlphaInterp.Result
    ExecutePin=@n1.ExecutePin
   )
   n2 = &RigVMFunction_AlphaInterp(
    InterpSpeedDecreasing=5
    InterpSpeedIncreasing=5
    Scale=1
    bClampResult=false
    Value=&RigVMFunction_MathBoolToFloat(Value=RigVMVariableNode() {variable=Pub_baseballsliding}.Value).Result
   )
   > n3 = &RigUnit_ModifyTransforms(
    ItemToModify="((Item=(Type=Bone,Name=\"dyn_backFlap_1\"),Transform=(Rotation=(X=0,Y=0,Z=-0.5,W=0.866),Translation=(X=0,Y=0,Z=0),Scale3D=(X=1,Y=1,Z=1))),(Item=(Type=Bone,Name=\"dyn_backFlap_2\"),Transform=(Rotation=(X=0,Y=0,Z=-0.1736,W=0.9848),Translation=(X=0,Y=0,Z=0),Scale ...(truncated)"
    WeightMinimum=0
    Weight=@n2.Result
   )
   RigVMFunctionReturnNode(ExecutePin=@n3.ExecutePin)
  ]}
 CollapseNode_ContainedGraph {nodes [
   n1 = RigVMFunctionEntryNode()
   &RigUnit_ParentConstraint(
    AdvancedSettings=(InterpolationType=Average,RotationOrderForFilter=XZY)
    Child=(Type=Bone,Name="dyn_neck_seat")
    Parents=((Item=(Type=Bone,Name="head"),Weight=1))
    Weight=1
    ExecutePin=@n1.ExecutePin
   )
   > &RigUnit_ParentConstraint(
    AdvancedSettings=(InterpolationType=Average,RotationOrderForFilter=XZY)
    Child=(Type=Bone,Name="dyn_neck_seat1")
    Parents=((Item=(Type=Bone,Name="head"),Weight=1))
    Weight=1
   )
   RigVMFunction_MathBoolToFloat = &RigVMFunction_MathBoolToFloat(Value=RigVMVariableNode() {variable=Pub_CelebrationEmote}.Value)
   AlphaInterp = &RigVMFunction_AlphaInterp(InterpSpeedDecreasing=5, InterpSpeedIncreasing=5, Scale=1, bClampResult=False, Value=@RigVMFunction_MathBoolToFloat.Result)
   > &RigUnit_ModifyTransforms(
    ItemToModify=((Item=(Type=Bone,Name="dyn_neck_seat"),Transform=(Rotation=(X=0,Y=0,Z=0.2588,W=0.9659),Translation=(X=5,Y=-5,Z=0),Scale3D=(X=1,Y=1,Z=1))))
    WeightMinimum=0
    Weight=@AlphaInterp.Result
   )
   n2 = &RigVMFunction_AlphaInterp(
    InterpSpeedDecreasing=5
    InterpSpeedIncreasing=5
    Scale=1
    bClampResult=False
    Value=&RigVMFunction_MathBoolToFloat(Value=RigVMVariableNode() {variable=Pub_ParachuteOpened}.Value).Result
   )
   > &RigUnit_ModifyTransforms(
    ItemToModify="((Item=(Type=Bone,Name=\"dyn_L_pauldron_bottom\"),Transform=(Rotation=(X=0,Y=0,Z=0,W=1),Translation=(X=0,Y=-5,Z=0),Scale3D=(X=1,Y=1,Z=1))),(Item=(Type=Bone,Name=\"dyn_R_pauldron_bottom\"),Transform=(Rotation=(X=0,Y=0,Z=0,W=1),Translation=(X=0,Y=5,Z=0),Scale3D= ...(truncated)"
    WeightMinimum=0
    Weight=@n2.Result
   )
   > DeltaRotation v4 = RigVMFunctionReferenceNode(
    Rotation Order=ZYX
    child=(Type=Bone,Name="neck_01")
    debugRotation=false
    outputMinMax=(X=0,Y=1)
    parent=(Type=Null,Name="neck_01_null")
    premult=1
    rotationMinMax=(X=0,Y=-20)
    triggerAxis=Z
    twistAxis=(X=1,Y=0,Z=0)
   )
   > &RigUnit_ModifyTransforms(
    ItemToModify=((Item=(Type=Bone,Name="dyn_neck_seat"),Transform=(Rotation=(X=0,Y=0,Z=0.2588,W=0.9659),Translation=(X=0,Y=0,Z=0),Scale3D=(X=1,Y=1,Z=1))))
    WeightMinimum=0
    Weight=@"DeltaRotation v4".outputWeight
   )
   > &RigUnit_ParentConstraint(
    AdvancedSettings=(InterpolationType=Average,RotationOrderForFilter=XZY)
    Child=(Type=Bone,Name="deform_lowerarm_gauntles_l")
    Parents=((Item=(Type=Bone,Name="deform_lowerarm_l"),Weight=1))
    Weight=1
   )
   > n3 = &RigUnit_ParentConstraint(
    AdvancedSettings=(InterpolationType=Average,RotationOrderForFilter=XZY)
    Child=(Type=Bone,Name="deform_lowerarm_gauntles_r")
    Parents=((Item=(Type=Bone,Name="deform_lowerarm_r"),Weight=1))
    Weight=1
   )
   RigVMFunctionReturnNode(ExecutePin=@n3.ExecutePin)
  ]}
 New Function_ContainedGraph {nodes [
   n1 = RigVMFunctionEntryNode()
   n2 = &RigUnit_GetTransform(bInitial=False, Item=@n1.Start)
   Subtract = /Script/RigVM.RigVMFunction_MathVectorSub(B=@n2.Transform.Translation, A=&RigUnit_GetTransform(bInitial=False, Item=@n1.End).Transform.Translation) {is=struct RigVMUnitNode, template="Subtract::Execute(in A,in B,out Result)"}
   Unit = /Script/RigVM.RigVMFunction_MathVectorUnit(Value=@Subtract.Result) {is=struct RigVMUnitNode, template="Unit::Execute(in Value,out Result)"}
   RigVMFunction_MathQuaternionFromAxisAndAngle = /Script/RigVM.RigVMFunction_MathQuaternionFromAxisAndAngle(Axis=@Unit.Result, Angle=@n1.QuatAmount) {is=struct RigVMUnitNode, template="FromAxisAndAngle::Execute(in Axis,in Angle,out Result)"}
   Multiply = /Script/RigVM.RigVMFunction_MathQuaternionMul(A=@RigVMFunction_MathQuaternionFromAxisAndAngle.Result, B=@n2.Transform.Rotation) {is=struct RigVMUnitNode, template="Multiply::Execute(in A,in B,out Result)"}
   > &RigUnit_SetRotation(Space=GlobalSpace, bInitial=False, Item=@n1.Start, Weight=@n1.Weight, Value=@Multiply.Result)
   > RigVMFunctionReturnNode()
  ]}
 New Function_ContainedGraph {nodes [
   RigVMFunctionEntryNode()
   n1 = &RigVMFunction_MathQuaternionToEuler(
    RotationOrder=ZXY
    Value=&RigUnit_GetRelativeTransformForItem(Child=(Type=Bone,Name="foot_l"), Parent=(Type=Bone,Name="calf_l"), bChildInitial=False, bParentInitial=False).RelativeTransform.Rotation
   )
   > RigVMVariableNode(Value=&RigVMFunction_MathDoubleRemap(SourceMaximum=-30, SourceMinimum=-5, TargetMaximum=1, TargetMinimum=0, Value=@n1.Result.Z).Result) {variable=L footBentUp alpha}
   > RigVMVariableNode(Value=&RigVMFunction_MathDoubleRemap(SourceMaximum=45, SourceMinimum=0, TargetMaximum=1, TargetMinimum=0, Value=@n1.Result.Z).Result) {variable=L footBentBack alpha}
   n2 = &RigUnit_GetRelativeTransformForItem(Child=(Type=Bone,Name="calf_l"), Parent=(Type=Bone,Name="thigh_l"), bChildInitial=False, bParentInitial=False)
   > RigVMVariableNode(Value=&RigVMFunction_MathDoubleRemap(
     SourceMaximum=115
     SourceMinimum=70
     TargetMaximum=1
     TargetMinimum=0
     Value=&RigVMFunction_MathQuaternionToEuler(RotationOrder=ZXY, Value=@n2.RelativeTransform.Rotation).Result.Z
    ).Result) {variable=L calfBent alpha}
   n3 = &RigVMFunction_MathQuaternionToEuler(
    RotationOrder=ZXY
    Value=&RigUnit_GetRelativeTransformForItem(Child=(Type=Bone,Name="foot_r"), Parent=(Type=Bone,Name="calf_r"), bChildInitial=False, bParentInitial=False).RelativeTransform.Rotation
   )
   > RigVMVariableNode(Value=&RigVMFunction_MathDoubleRemap(SourceMaximum=-30, SourceMinimum=-5, TargetMaximum=1, TargetMinimum=0, Value=@n3.Result.Z).Result) {variable=R footBentUp alpha}
   > RigVMVariableNode(Value=&RigVMFunction_MathDoubleRemap(SourceMaximum=45, SourceMinimum=0, TargetMaximum=1, TargetMinimum=0, Value=@n3.Result.Z).Result) {variable=R footBentBack alpha}
   n4 = &RigUnit_GetRelativeTransformForItem(Child=(Type=Bone,Name="calf_r"), Parent=(Type=Bone,Name="thigh_r"), bChildInitial=False, bParentInitial=False)
   > RigVMVariableNode(Value=&RigVMFunction_MathDoubleRemap(
     SourceMaximum=115
     SourceMinimum=70
     TargetMaximum=1
     TargetMinimum=0
     Value=&RigVMFunction_MathQuaternionToEuler(RotationOrder=ZXY, Value=@n4.RelativeTransform.Rotation).Result.Z
    ).Result) {variable=R calfBent alpha}
   n5 = &RigUnit_GetRelativeTransformForItem(Child=(Type=Bone,Name="lowerarm_l"), Parent=(Type=Bone,Name="upperarm_l"), bChildInitial=False, bParentInitial=False)
   > RigVMVariableNode(Value=&RigVMFunction_MathDoubleRemap(
     SourceMaximum=110
     SourceMinimum=70
     TargetMaximum=1
     TargetMinimum=0
     Value=&RigVMFunction_MathQuaternionToEuler(RotationOrder=ZXY, Value=@n5.RelativeTransform.Rotation).Result.Z
    ).Result) {variable=L armBend alpha}
   n6 = &RigUnit_GetRelativeTransformForItem(Child=(Type=Bone,Name="lowerarm_r"), Parent=(Type=Bone,Name="upperarm_r"), bChildInitial=False, bParentInitial=False)
   > RigVMVariableNode(Value=&RigVMFunction_MathDoubleRemap(
     SourceMaximum=110
     SourceMinimum=70
     TargetMaximum=1
     TargetMinimum=0
     Value=&RigVMFunction_MathQuaternionToEuler(RotationOrder=ZXY, Value=@n6.RelativeTransform.Rotation).Result.Z
    ).Result) {variable=R armBend alpha}
   > RigVMFunctionReturnNode()
  ]}
 New Function_ContainedGraph {nodes [
   RigVMFunctionEntryNode()
   > &RigUnit_ModifyTransforms(
    ItemToModify=((Item=(Type=Bone,Name="dyn_L_upperBack_scaler"),Transform=(Rotation=(X=0,Y=0,Z=0,W=1),Translation=(X=0,Y=0,Z=0),Scale3D=(X=0.3,Y=1,Z=1))))
    WeightMinimum=0
    Weight=RigVMVariableNode() {variable=L calfBent alpha}.Value
   )
   > &RigUnit_ModifyTransforms(
    ItemToModify=((Item=(Type=Bone,Name="dyn_L_lowerFront_scaler"),Transform=(Rotation=(X=0,Y=0,Z=0,W=1),Translation=(X=0,Y=0,Z=0),Scale3D=(X=0.3,Y=1,Z=1))))
    WeightMinimum=0
    Weight=RigVMVariableNode() {variable=L footBentUp alpha}.Value
   )
   > &RigUnit_ModifyTransforms(
    ItemToModify=((Item=(Type=Bone,Name="dyn_L_lowerBack_scaler"),Transform=(Rotation=(X=0,Y=0,Z=0,W=1),Translation=(X=0,Y=0,Z=0),Scale3D=(X=0.5,Y=1,Z=1))))
    WeightMinimum=0
    Weight=RigVMVariableNode() {variable=L footBentBack alpha}.Value
   )
   > &RigUnit_ModifyTransforms(
    ItemToModify=((Item=(Type=Bone,Name="dyn_R_upperBack_scaler"),Transform=(Rotation=(X=0,Y=0,Z=0,W=1),Translation=(X=0,Y=0,Z=0),Scale3D=(X=0.3,Y=1,Z=1))))
    WeightMinimum=0
    Weight=RigVMVariableNode() {variable=R calfBent alpha}.Value
   )
   > &RigUnit_ModifyTransforms(
    ItemToModify=((Item=(Type=Bone,Name="dyn_R_lowerFront_scaler"),Transform=(Rotation=(X=0,Y=0,Z=0,W=1),Translation=(X=0,Y=0,Z=0),Scale3D=(X=0.3,Y=1,Z=1))))
    WeightMinimum=0
    Weight=RigVMVariableNode() {variable=R footBentUp alpha}.Value
   )
   > &RigUnit_ModifyTransforms(
    ItemToModify=((Item=(Type=Bone,Name="dyn_R_lowerBack_scaler"),Transform=(Rotation=(X=0,Y=0,Z=0,W=1),Translation=(X=0,Y=0,Z=0),Scale3D=(X=0.5,Y=1,Z=1))))
    WeightMinimum=0
    Weight=RigVMVariableNode() {variable=R footBentBack alpha}.Value
   )
   > &RigUnit_ModifyTransforms(
    ItemToModify=((Item=(Type=Bone,Name="dyn_L_armScaler"),Transform=(Rotation=(X=0,Y=0,Z=0,W=1),Translation=(X=-2.5,Y=0,Z=0),Scale3D=(X=0.55,Y=1,Z=1))))
    WeightMinimum=0
    Weight=RigVMVariableNode() {variable=L armBend alpha}.Value
   )
   > &RigUnit_ModifyTransforms(
    ItemToModify=((Item=(Type=Bone,Name="dyn_R_armScaler"),Transform=(Rotation=(X=0,Y=0,Z=0,W=1),Translation=(X=2.5,Y=0,Z=0),Scale3D=(X=0.55,Y=1,Z=1))))
    WeightMinimum=0
    Weight=RigVMVariableNode() {variable=R armBend alpha}.Value
   )
   > RigVMFunctionReturnNode()
  ]}
 New Function_ContainedGraph {nodes [
   n1 = RigVMFunctionEntryNode()
   RigVMFunction_ControlFlowBranch = &RigVMFunction_ControlFlowBranch(ExecuteContext=@n1.ExecuteContext, Condition=RigVMVariableNode() {variable=Pub_IsStickOfTruth}.Value)
   RigVMFunctionReturnNode(ExecuteContext=@RigVMFunction_ControlFlowBranch.Completed)
   n2 = &RigUnit_SphericalPoseReader(
    ActiveRegionScaleFactors=(PositiveWidth=5,NegativeWidth=5,PositiveHeight=1.5,NegativeHeight=15)
    ActiveRegionSize=0.04
    Debug=(bDrawDebug=false,bDrawLocalAxes=false,DebugScale=25,DebugSegments=20,DebugThickness=0.25)
    DriverAxis=(X=1,Y=0,Z=0)
    DriverItem=(Type=Bone,Name="upperarm_l")
    FalloffRegionScaleFactors=(PositiveWidth=2.5,NegativeWidth=1,PositiveHeight=1.5,NegativeHeight=1)
    FalloffSize=0.3
    FlipWidthScaling=false
    OptionalParentItem=(Type=Bone,Name="spine_05")
    RotationOffset=(X=60,Y=40,Z=0)
    ExecutePin=@RigVMFunction_ControlFlowBranch.False
   )
   > RigVMVariableNode(Value=@n2.OutputParam) {variable=L armForward}
   > n3 = &RigUnit_SphericalPoseReader(
    ActiveRegionScaleFactors=(PositiveWidth=1,NegativeWidth=4,PositiveHeight=2,NegativeHeight=2)
    ActiveRegionSize=0.1
    Debug=(bDrawDebug=false,bDraw2D=False,bDrawLocalAxes=False,DebugScale=25,DebugSegments=20,DebugThickness=0.25)
    DriverAxis=(X=1,Y=0,Z=0)
    DriverItem=(Type=Bone,Name="upperarm_l")
    FalloffRegionScaleFactors=(PositiveWidth=0.5,NegativeWidth=1,PositiveHeight=0.5,NegativeHeight=0.5)
    FalloffSize=0.4
    FlipHeightScaling=False
    FlipWidthScaling=False
    OptionalParentItem=(Type=Bone,Name="None")
    RotationOffset=(X=0,Y=-50,Z=0)
   )
   n4 = RigVMVariableNode() {variable=L armForward}
   > &RigUnit_ModifyTransforms(
    ItemToModify=((Item=(Type=Bone,Name="dyn_L_chairScaler"),Transform=(Rotation=(X=0,Y=0,Z=0,W=1),Translation=(X=0,Y=0,Z=0),Scale3D=(X=1,Y=1,Z=0.7))))
    WeightMinimum=0
    Weight=@n4.Value
   )
   > &RigUnit_ModifyTransforms(
    ItemToModify=((Item=(Type=Bone,Name="dyn_L_pauldron_bottom"),Transform=(Rotation=(X=0,Y=0,Z=0,W=1),Translation=(X=-3,Y=0,Z=0),Scale3D=(X=0.9,Y=0.9,Z=1))))
    WeightMinimum=0
    Weight=@n3.OutputParam
   )
   > &RigUnit_ModifyTransforms(
    ItemToModify="((Item=(Type=Bone,Name=\"dyn_L_cardboardFlap_A\"),Transform=(Rotation=(X=0,Y=-0.6088,Z=0,W=0.7934),Translation=(X=0,Y=0,Z=0),Scale3D=(X=1,Y=1,Z=1))),(Item=(Type=Bone,Name=\"dyn_L_cardboardFlap_B\"),Transform=(Rotation=(X=0,Y=-0.6088,Z=0,W=0.7934),Translation=( ...(truncated)"
    WeightMinimum=0
    Weight=@n4.Value
   )
   > &RigUnit_ModifyTransforms(
    ItemToModify=((Item=(Type=Bone,Name="dyn_L_pauldron_top"),Transform=(Rotation=(X=0.4226,Y=0,Z=0,W=0.9063),Translation=(X=0.5,Y=-5,Z=0),Scale3D=(X=0.9,Y=1,Z=1))))
    WeightMinimum=0
    Weight=@n4.Value
   )
   > n5 = &RigUnit_SphericalPoseReader(
    ActiveRegionScaleFactors=(PositiveWidth=4,NegativeWidth=1,PositiveHeight=1,NegativeHeight=1)
    ActiveRegionSize=0.1
    Debug=(bDrawDebug=false,bDraw2D=False,bDrawLocalAxes=False,DebugScale=25,DebugSegments=20,DebugThickness=0.25)
    DriverAxis=(X=1,Y=0,Z=0)
    DriverItem=(Type=Bone,Name="upperarm_l")
    FalloffRegionScaleFactors=(PositiveWidth=1,NegativeWidth=0.4,PositiveHeight=0.5,NegativeHeight=0.5)
    FalloffSize=0.35
    FlipHeightScaling=false
    FlipWidthScaling=true
    OptionalParentItem=(Type=Bone,Name="spine_05")
    RotationOffset=(X=0,Y=-10,Z=0)
   )
   > &RigUnit_ModifyTransforms(
    ItemToModify=((Item=(Type=Bone,Name="dyn_L_pauldron_bottom"),Transform=(Rotation=(X=0,Y=-0.0872,Z=0,W=0.9962),Translation=(X=-5,Y=0,Z=6),Scale3D=(X=0.8,Y=1,Z=1))))
    WeightMinimum=0
    Weight=@n5.OutputParam
   )
   > n6 = &RigUnit_SphericalPoseReader(
    ActiveRegionScaleFactors=(PositiveWidth=5,NegativeWidth=5,PositiveHeight=1.5,NegativeHeight=15)
    ActiveRegionSize=0.04
    Debug=(bDrawDebug=false,bDraw2D=false,bDrawLocalAxes=false,DebugScale=25,DebugSegments=20,DebugThickness=0.25)
    DriverAxis=(X=-1,Y=0,Z=0)
    DriverItem=(Type=Bone,Name="upperarm_r")
    FalloffRegionScaleFactors=(PositiveWidth=1,NegativeWidth=2.5,PositiveHeight=1.5,NegativeHeight=1)
    FalloffSize=0.3
    FlipHeightScaling=true
    FlipWidthScaling=true
    OptionalParentItem=(Type=Bone,Name="spine_05")
    RotationOffset=(X=120,Y=-40,Z=180)
   )
   > RigVMVariableNode(Value=@n6.OutputParam) {variable=R armForward}
   > n7 = &RigUnit_SphericalPoseReader(
    ActiveRegionScaleFactors=(PositiveWidth=4,NegativeWidth=1,PositiveHeight=2,NegativeHeight=2)
    ActiveRegionSize=0.1
    Debug=(bDrawDebug=false,bDraw2D=False,bDrawLocalAxes=False,DebugScale=25,DebugSegments=20,DebugThickness=0.25)
    DriverAxis=(X=-1,Y=0,Z=0)
    DriverItem=(Type=Bone,Name="upperarm_r")
    FalloffRegionScaleFactors=(PositiveWidth=1,NegativeWidth=0.5,PositiveHeight=0.5,NegativeHeight=0.5)
    FalloffSize=0.4
    FlipHeightScaling=false
    FlipWidthScaling=true
    OptionalParentItem=(Type=Bone,Name="None")
    RotationOffset=(X=180,Y=50,Z=180)
   )
   n8 = RigVMVariableNode() {variable=R armForward}
   > &RigUnit_ModifyTransforms(
    ItemToModify=((Item=(Type=Bone,Name="dyn_R_chairScaler"),Transform=(Rotation=(X=0,Y=0,Z=0,W=1),Translation=(X=0,Y=0,Z=0),Scale3D=(X=1,Y=1,Z=0.7))))
    WeightMinimum=0
    Weight=@n8.Value
   )
   > &RigUnit_ModifyTransforms(
    ItemToModify=((Item=(Type=Bone,Name="dyn_R_pauldron_bottom"),Transform=(Rotation=(X=0,Y=0,Z=0,W=1),Translation=(X=3,Y=0,Z=0),Scale3D=(X=0.9,Y=0.9,Z=1))))
    WeightMinimum=0
    Weight=@n7.OutputParam
   )
   > &RigUnit_ModifyTransforms(
    ItemToModify="((Item=(Type=Bone,Name=\"dyn_R_cardboardFlap_A\"),Transform=(Rotation=(X=0,Y=-0.6088,Z=0,W=0.7934),Translation=(X=0,Y=0,Z=0),Scale3D=(X=1,Y=1,Z=1))),(Item=(Type=Bone,Name=\"dyn_R_cardboardFlap_B\"),Transform=(Rotation=(X=0,Y=-0.6088,Z=0,W=0.7934),Translation=( ...(truncated)"
    WeightMinimum=0
    Weight=@n8.Value
   )
   > &RigUnit_ModifyTransforms(
    ItemToModify=((Item=(Type=Bone,Name="dyn_R_pauldron_top"),Transform=(Rotation=(X=0.4226,Y=0,Z=0,W=0.9063),Translation=(X=-0.5,Y=5,Z=0),Scale3D=(X=0.9,Y=1,Z=1))))
    WeightMinimum=0
    Weight=@n8.Value
   )
   > n9 = &RigUnit_SphericalPoseReader(
    ActiveRegionScaleFactors=(PositiveWidth=4,NegativeWidth=1,PositiveHeight=1,NegativeHeight=1)
    ActiveRegionSize=0.1
    Debug=(bDrawDebug=false,bDraw2D=False,bDrawLocalAxes=False,DebugScale=25,DebugSegments=20,DebugThickness=0.25)
    DriverAxis=(X=-1,Y=0,Z=0)
    DriverItem=(Type=Bone,Name="upperarm_r")
    FalloffRegionScaleFactors=(PositiveWidth=1,NegativeWidth=0.4,PositiveHeight=0.5,NegativeHeight=0.5)
    FalloffSize=0.4
    FlipHeightScaling=false
    FlipWidthScaling=true
    OptionalParentItem=(Type=Bone,Name="spine_05")
    RotationOffset=(X=180,Y=10,Z=180)
   )
   > &RigUnit_ModifyTransforms(
    ItemToModify=((Item=(Type=Bone,Name="dyn_R_pauldron_bottom"),Transform=(Rotation=(X=0,Y=-0.0872,Z=0,W=0.9962),Translation=(X=5,Y=0,Z=-6),Scale3D=(X=0.8,Y=1,Z=1))))
    WeightMinimum=0
    Weight=@n9.OutputParam
   )
   DeltaRotation v4 = RigVMFunctionReferenceNode(
    Rotation Order=ZYX
    child=(Type=Bone,Name="deform_upperarm_r")
    debugRotation=false
    outputMinMax=(X=0,Y=1)
    parent=(Type=Null,Name="upperarm_r_null")
    premult=1
    rotationMinMax=(X=0,Y=-55)
    triggerAxis=Y
    twistAxis=(X=1,Y=0,Z=0)
    ExecuteContext=@RigVMFunction_ControlFlowBranch.True
   )
   AlphaInterp = &RigVMFunction_AlphaInterp(InterpSpeedDecreasing=5, InterpSpeedIncreasing=5, Scale=1, bClampResult=False, Value=@"DeltaRotation v4".outputWeight)
   > &RigUnit_ModifyTransforms(
    ItemToModify=((Item=(Type=Bone,Name="dyn_R_pauldron_top"),Transform=(Rotation=(X=0.1632,Y=-0.0594,Z=-0.3368,W=0.9254),Translation=(X=2,Y=7,Z=0),Scale3D=(X=1,Y=1,Z=1))))
    WeightMinimum=0
    Weight=@AlphaInterp.Result
   )
  ]}
 New Function_ContainedGraph {nodes [
   RigVMFunctionEntryNode()
   > RigVMFunction_Sequence = RigVMAggregateNode()
   &RigUnit_AimConstraintLocalSpaceOffset(
    Child=(Type=Bone,Name="deform_upperarm_l")
    Filter=(bX=True,bY=True,bZ=True)
    Parents=((Item=(Type=Bone,Name="lowerarm_l"),Weight=1))
    Weight=1
    WorldUp=(Target=(X=0,Y=0,Z=1),Kind=Location,Space=(Type=Bone,Name="lowerarm_l"))
    ExecutePin=@RigVMFunction_Sequence.B
   )
   > &RigUnit_ParentConstraint(
    AdvancedSettings=(InterpolationType=Average,RotationOrderForFilter=XZY)
    Child=(Type=Bone,Name="deform_lowerarm_l")
    Parents=((Item=(Type=Bone,Name="lowerarm_l"),Weight=1))
    Weight=1
   )
   > &RigUnit_ParentConstraint(
    AdvancedSettings=(InterpolationType=Average,RotationOrderForFilter=XZY)
    Child=(Type=Bone,Name="deform_hand_l")
    Parents=((Item=(Type=Bone,Name="hand_l"),Weight=1))
    Weight=1
   )
   > &RigUnit_AimConstraintLocalSpaceOffset(
    Child=(Type=Bone,Name="deform_upperarm_r")
    Filter=(bX=True,bY=True,bZ=True)
    Parents=((Item=(Type=Bone,Name="cr_lowerarm_push_r"),Weight=1))
    Weight=1
    WorldUp=(Target=(X=0,Y=0,Z=1),Kind=Location,Space=(Type=Bone,Name="cr_lowerarm_push_r"))
   )
   > &RigUnit_ParentConstraint(
    AdvancedSettings=(InterpolationType=Average,RotationOrderForFilter=XZY)
    Child=(Type=Bone,Name="deform_lowerarm_r")
    Parents=((Item=(Type=Bone,Name="cr_lowerarm_push_r"),Weight=1))
    Weight=1
   )
   > &RigUnit_ParentConstraint(
    AdvancedSettings=(InterpolationType=Average,RotationOrderForFilter=XZY)
    Child=(Type=Bone,Name="deform_hand_r")
    Parents=((Item=(Type=Bone,Name="hand_r"),Weight=1))
    Weight=1
   )
   /Script/PBIK.RigUnit_PBIK(
    BoneSettings=((Bone="deform_hand_l",RotationStiffness=0,PositionStiffness=0,X=Free,MinX=0,MaxX=0,Y=Free,MinY=0,MaxY=0,Z=Free,MinZ=0,MaxZ=0,bUsePreferredAngles=true,PreferredAngles=(X=1,Y=0,Z=0)))
    Debug=(DrawScale=1,bDrawDebug=False)
    Effectors=((Bone="hand_l",Transform=(Rotation=(X=0,Y=0,Z=0,W=1),Translation=(X=0,Y=0,Z=0),Scale3D=(X=1,Y=1,Z=1)),PositionAlpha=1,RotationAlpha=1,StrengthAlpha=1,PullChainAlpha=1,PinRotation=1,ChainDepth=0))
    ExcludedBones=()
    Root=deform_upperarm_l
    Settings="(Iterations=20,SubIterations=0,MassMultiplier=1,bAllowStretch=false,RootBehavior=PrePull,PrePullRootSettings=(RotationAlpha=0,RotationAlphaX=1,RotationAlphaY=1,RotationAlphaZ=1,PositionAlpha=1,PositionAlphaX=1,PositionAlphaY=1,PositionAlphaZ=1),GlobalPullC ...(truncated)"
    ExecutePin=@RigVMFunction_Sequence.C
   ) {is=struct RigVMUnitNode}
   > &RigUnit_PBIK(
    BoneSettings="((Bone=\"deform_hand_r\",RotationStiffness=0,PositionStiffness=0,X=Free,MinX=0,MaxX=0,Y=Free,MinY=0,MaxY=0,Z=Free,MinZ=0,MaxZ=0,bUsePreferredAngles=false,PreferredAngles=(X=1,Y=0,Z=0)),(Bone=\"deform_upperarm_l\",RotationStiffness=0,PositionStiffness=0,X=Free, ...(truncated)"
    Debug=(DrawScale=1,bDrawDebug=False)
    Effectors="((Bone=\"hand_r\",Transform=(Rotation=(X=0,Y=0,Z=0,W=1),Translation=(X=0,Y=0,Z=0),Scale3D=(X=1,Y=1,Z=1)),PositionAlpha=1,RotationAlpha=1,StrengthAlpha=1,PullChainAlpha=1,PinRotation=1,ChainDepth=0),(Bone=\"upperarm_r\",Transform=(Rotation=(X=0,Y=0,Z=0,W=1),Tra ...(truncated)"
    ExcludedBones=()
    Root=deform_upperarm_r
    Settings="(Iterations=20,SubIterations=0,MassMultiplier=1,bAllowStretch=true,RootBehavior=PrePull,PrePullRootSettings=(RotationAlpha=0,RotationAlphaX=1,RotationAlphaY=1,RotationAlphaZ=1,PositionAlpha=1,PositionAlphaX=1,PositionAlphaY=1,PositionAlphaZ=1),GlobalPullCh ...(truncated)"
   )
   > FBIK Chain Correction_1 = RigVMFunctionReferenceNode(
    End=(Type=Bone,Name="deform_hand_r")
    QuatAmount=2.5
    Start=(Type=Bone,Name="deform_upperarm_r")
    Weight=RigVMVariableNode() {variable=R Elbow Distance to Head}.Value
   )
   > FBIK Chain Correction = RigVMFunctionReferenceNode(
    End=(Type=Bone,Name="deform_hand_l")
    QuatAmount=2.5
    Start=(Type=Bone,Name="deform_upperarm_l")
    Weight=RigVMVariableNode() {variable=L Elbow Distance to Head}.Value
   )
   Deform Fingers = RigVMDispatchNode(Value="((Type=Bone,Name=\"deform_thumb_01_l\"),(Type=Bone,Name=\"deform_thumb_02_l\"),(Type=Bone,Name=\"deform_thumb_03_l\"),(Type=Bone,Name=\"deform_index_metacarpal_l\"),(Type=Bone,Name=\"deform_index_01_l\"),(Type=Bone,Name=\"deform_index_02_l\"),(Type=Bone,Name=\"deform_i ...(truncated)") {template="DISPATCH_RigVMDispatch_Constant(io Value)"}
   For_Each_7 = RigVMDispatchNode(Array=@"Deform Fingers".Value, ExecuteContext=@RigVMFunction_Sequence.D) {template="DISPATCH_RigVMDispatch_ArrayIterator(in Array,out Element,out Index,out Count,out Ratio)"}
   Base Fingers = RigVMDispatchNode(Value="((Type=Bone,Name=\"thumb_01_l\"),(Type=Bone,Name=\"thumb_02_l\"),(Type=Bone,Name=\"thumb_03_l\"),(Type=Bone,Name=\"index_metacarpal_l\"),(Type=Bone,Name=\"index_01_l\"),(Type=Bone,Name=\"index_02_l\"),(Type=Bone,Name=\"index_03_l\"),(Type=Bone,Name=\"middle_metacarpal_l\" ...(truncated)") {template="DISPATCH_RigVMDispatch_Constant(io Value)"}
   > &RigUnit_ParentConstraint(
    AdvancedSettings=(InterpolationType=Average,RotationOrderForFilter=XZY)
    Parents=((Item=(Type=Bone,Name="None"),Weight=1))
    Weight=1
    Parents.0.Item=RigVMDispatchNode(Array=@"Base Fingers".Value, Index=@For_Each_7.Index) {template="DISPATCH_RigVMDispatch_ArrayGetAtIndex(in Array,in Index,out Element)"}.Element
    Child=@For_Each_7.Element
   )
   RigVMFunctionReturnNode(ExecuteContext=@RigVMFunction_Sequence.E)
   Get Normalized Distance Between_1_1 = RigVMFunctionReferenceNode(
    Debug=false
    Max Distance=35
    Min Distance=1
    Source=(Type=Bone,Name="lowerarm_r")
    Target=(Type=Bone,Name="head")
    ExecuteContext=@RigVMFunction_Sequence.A
   )
   > RigVMVariableNode(Value=@"Get Normalized Distance Between_1_1".Result) {variable=R Elbow Distance to Head}
   > &RigUnit_ModifyTransforms(
    ItemToModify=((Item=(Type=Bone,Name="cr_lowerarm_push_r"),Transform=(Rotation=(X=0,Y=0,Z=0,W=1),Translation=(X=0,Y=7,Z=0),Scale3D=(X=1,Y=1,Z=1))))
    WeightMinimum=0
    Weight=RigVMVariableNode() {variable=R Elbow Distance to Head}.Value
   )
   > &RigUnit_AimConstraintLocalSpaceOffset(
    Child=(Type=Bone,Name="cr_lowerarm_push_r")
    Filter=(bX=True,bY=True,bZ=True)
    Parents=((Item=(Type=Bone,Name="hand_r"),Weight=1))
    Weight=1
    WorldUp=(Target=(X=0,Y=0,Z=1),Kind=Location,Space=(Type=Bone,Name="hand_r"))
   )
   > Get Normalized Distance Between_1_2 = RigVMFunctionReferenceNode(Debug=false, Max Distance=25, Min Distance=1, Source=(Type=Bone,Name="lowerarm_l"), Target=(Type=Bone,Name="head"))
   > RigVMVariableNode(Value=@"Get Normalized Distance Between_1_2".Result) {variable=L Elbow Distance to Head}
   > &RigUnit_ModifyTransforms(
    ItemToModify=((Item=(Type=Bone,Name="cr_lowerarm_push_l"),Transform=(Rotation=(X=0,Y=0,Z=0,W=1),Translation=(X=0,Y=7,Z=0),Scale3D=(X=1,Y=1,Z=1))))
    WeightMinimum=0
    Weight=RigVMVariableNode() {variable=L Elbow Distance to Head}.Value
   )
   > &RigUnit_AimConstraintLocalSpaceOffset(
    Child=(Type=Bone,Name="cr_lowerarm_push_l")
    Filter=(bX=True,bY=True,bZ=True)
    Parents=((Item=(Type=Bone,Name="hand_l"),Weight=1))
    Weight=1
    WorldUp=(Target=(X=0,Y=0,Z=1),Kind=Location,Space=(Type=Bone,Name="hand_l"))
   )
   Get Normalized Distance Between_1_3 = RigVMFunctionReferenceNode(Debug=false, ExecuteContext=(), Max Distance=40, Min Distance=1, Source=(Type=Bone,Name="hand_l"), Target=(Type=Bone,Name="head"))
  ]}
 New Function_ContainedGraph {nodes [
   n1 = RigVMFunctionEntryNode()
   RigVMFunction_MathBoolToInteger = &RigVMFunction_MathBoolToInteger(Value=@n1.onlyTwist)
   RigVMFunction_MathQuaternionSwingTwist = &RigVMFunction_MathQuaternionSwingTwist(
    TwistAxis=@n1.twistAxis
    Input=&RigUnit_GetRelativeTransformForItem(bChildInitial=False, bParentInitial=False, Child=@n1.child, Parent=@n1.parent).RelativeTransform.Rotation
   )
   n2 = RigVMDispatchNode(
    Values=((),())
    Index=@RigVMFunction_MathBoolToInteger.Result
    Values.0=@RigVMFunction_MathQuaternionSwingTwist.Swing
    Values.1=@RigVMFunction_MathQuaternionSwingTwist.Twist
   ) {template="DISPATCH_RigVMDispatch_SelectInt32(in Index,in Values,out Result)"}
   RigVMFunction_MathQuaternionToEuler = &RigVMFunction_MathQuaternionToEuler(Value=@n2.Result, RotationOrder=@n1.Rotation Order)
   > DISPATCH_RigVMDispatch_Print = RigVMDispatchNode(ScreenColor=(R=1,G=1,B=1,A=1), ScreenDuration=1.5, Enabled=@n1.debugRotation, Value=@RigVMFunction_MathQuaternionToEuler.Result) {template="DISPATCH_RigVMDispatch_Print(in Prefix,in Value,in Enabled,in ScreenDuration,in ScreenColor)"}
   Cast = RigVMDispatchNode(Value=@n1.triggerAxis) {template="DISPATCH_RigVMDispatch_CastEnumToInt(in Value,out Result)"}
   DISPATCH_RigVMDispatch_SelectInt32 = RigVMDispatchNode(
    Values=(0,0,0)
    Index=@Cast.Result
    Values.0=@RigVMFunction_MathQuaternionToEuler.Result.X
    Values.1=@RigVMFunction_MathQuaternionToEuler.Result.Y
    Values.2=@RigVMFunction_MathQuaternionToEuler.Result.Z
   ) {template="DISPATCH_RigVMDispatch_SelectInt32(in Index,in Values,out Result)"}
   Remap = &RigVMFunction_MathDoubleRemap(
    bClamp=False
    Value=@DISPATCH_RigVMDispatch_SelectInt32.Result
    TargetMinimum=@n1.outputMinMax.X
    TargetMaximum=@n1.outputMinMax.Y
    SourceMinimum=@n1.rotationMinMax.X
    SourceMaximum=@n1.rotationMinMax.Y
   )
   Absolute = &RigVMFunction_MathDoubleAbs(Value=@Remap.Result)
   Multiply = &RigVMFunction_MathDoubleMul(
    A=&RigVMFunction_MathDoubleClamp(Value=RigVMDispatchNode(True=@Absolute.Result, Condition=@n1.absoluteValue, False=@Remap.Result) {template="DISPATCH_RigVMDispatch_If(in Condition,in True,in False,out Result)"}.Result).Result
    B=@n1.premult
   )
   > RigVMFunctionReturnNode(outputQuat=@n2.Result, outputWeight=@Multiply.Result)
  ]}
 RigVMFunctionLibrary {nodes [
   Arm and Hand Retarget = RigVMCollapseNode(ExecuteContext=())
   ArmLeg Foil Adjustments = RigVMCollapseNode(ExecuteContext=())
   Delta Ref Variable Creation = RigVMCollapseNode(ExecuteContext=())
   DeltaRotation v4 = RigVMCollapseNode(
    ExecuteContext=()
    Rotation Order=ZYX
    child=(Type=Bone,Name="None")
    outputMinMax=(X=0,Y=1)
    parent=(Type=Bone,Name="None")
    premult=1
    rotationMinMax=(X=0,Y=0)
    twistAxis=(X=1,Y=0,Z=0)
   )
   FBIK Chain Correction = RigVMCollapseNode(End=(Type=None,Name="None"), ExecuteContext=(), Start=(Type=None,Name="None"))
   Pauldron Setup = RigVMCollapseNode(ExecuteContext=())
  ]}
 RigVMFunction_Sequence_1_ContainedGraph {nodes [
   n1 = RigVMFunctionEntryNode()
   RigVMFunction_Sequence = &RigVMFunction_Sequence(ExecuteContext=@n1.ExecuteContext)
   n2 = &RigVMFunction_Sequence(ExecuteContext=@RigVMFunction_Sequence.B)
   n3 = &RigVMFunction_Sequence(ExecuteContext=@n2.B)
   n4 = &RigVMFunction_Sequence(ExecuteContext=@n3.B)
   RigVMFunctionReturnNode(A=@RigVMFunction_Sequence.A, B=@n2.A, C=@n3.A, D=@n4.A, E=@n4.B)
  ]}
 RigVMModel {nodes [
   RigUnit_BeginExecution = &RigUnit_BeginExecution()
   Delta Ref Variable Creation = RigVMFunctionReferenceNode(ExecuteContext=@RigUnit_BeginExecution.ExecutePin)
   > ArmLeg Foil Adjustments = RigVMFunctionReferenceNode()
   > Arm and Hand Retarget = RigVMFunctionReferenceNode()
   > Pauldron Setup = RigVMFunctionReferenceNode()
   KneeSliding = RigVMCollapseNode(ExecutePin=@"Pauldron Setup".ExecuteContext)
   ExtraFixes = RigVMCollapseNode(ExecutePin=@KneeSliding.ExecutePin)
  ]}
]
hierarchy [
 root {loc=<0 0 0>, rot=<0 0 0 1>}
 pelvis {parent=root, loc=<0 -0.7321 92.4218>, rot=<0.0127 0.707 -0.0127 -0.707>}
 spine_01 {parent=pelvis, loc=<5.6002 -0.5128 0>, rot=<0 0 0.034 -0.9994>}
 spine_02 {parent=spine_01, loc=<8.0756 0.0139 0>, rot=<0 0 0.0238 -0.9997>}
 spine_03 {parent=spine_02, loc=<7.0395 0 0>, rot=<0 0 0.0149 -0.9999>}
 spine_04 {parent=spine_03, loc=<8.1023 0 0>, rot=<0 0 -0.0132 -0.9999>}
 spine_05 {parent=spine_04, loc=<10.0141 0 0>, rot=<0 0 -0.0015 -1>}
 clavicle_l {parent=spine_05, loc=<2.5856 -0.2866 -4.3777>, rot=<0.0465 -0.7289 -0.009 -0.6829>}
 upperarm_l {parent=clavicle_l, loc=<18.2264 0.1342 -0.001>, rot=<0.0253 -0.2693 0.0013 -0.9627>}
 lowerarm_l {parent=upperarm_l, loc=<29.0071 0.1686 -1.239>, rot=<0.0391 -0.0099 -0.2453 -0.9686>}
 hand_l {parent=lowerarm_l, loc=<25.8607 0.222 1.0274>, rot=<-0.7193 -0.0988 0.1193 -0.6773>}
 weapon_l {parent=hand_l, loc=<-0.1974 -3.4111 0.3811>, rot=<0 0 0 1>}
 hand_attach_l {parent=hand_l, loc=<8.5 -3 0>, rot=<0.7071 0.7071 0 0>}
 thumb_01_l {parent=hand_l, loc=<1.3118 -0.3191 -2.459>, rot=<0.6546 -0.4042 0.0664 -0.6354>}
 thumb_02_l {parent=thumb_01_l, loc=<5.9336 -0.0448 0.0401>, rot=<-0.0037 0.0213 0.298 -0.9543>}
 thumb_03_l {parent=thumb_02_l, loc=<2.2223 0.3682 -0.0907>, rot=<0.0118 -0.013 0.0066 -0.9998>}
 index_metacarpal_l {parent=hand_l, loc=<1.6535 0.0183 -2.96>, rot=<0.0158 -0.0124 -0.0173 -0.9996>}
 index_01_l {parent=index_metacarpal_l, loc=<8.3008 0.0357 -0.0524>, rot=<-0.0044 0.0485 0.0924 -0.9945>}
 index_02_l {parent=index_01_l, loc=<3.3587 -0.3004 0.0452>, rot=<-0.002 0.0027 0.3225 -0.9466>}
 index_03_l {parent=index_02_l, loc=<2.5578 -0.123 -0.0252>, rot=<-0.002 0.0067 0.2558 -0.9667>}
 middle_metacarpal_l {parent=hand_l, loc=<2.1856 -0.1725 -0.5984>, rot=<0.006 0.0168 -0.0329 -0.9993>}
 middle_01_l {parent=middle_metacarpal_l, loc=<7.3275 -0.2447 0.0129>, rot=<-0.0208 0.0042 0.0858 -0.9961>}
 middle_02_l {parent=middle_01_l, loc=<4.0123 0.1401 0.1175>, rot=<-0.0218 0.0258 0.3512 -0.9357>}
 middle_03_l {parent=middle_02_l, loc=<2.6077 0.0534 0.1373>, rot=<0.0082 0.0188 0.1708 -0.9851>}
 ring_metacarpal_l {parent=hand_l, loc=<2.2074 -0.3625 1.3638>, rot=<-0.0284 0.0722 -0.0458 -0.9959>}
 ring_01_l {parent=ring_metacarpal_l, loc=<7.0295 -0.6197 -0.1446>, rot=<-0.0489 -0.0252 0.1892 -0.9804>}
 ring_02_l {parent=ring_01_l, loc=<3.6174 0.4647 -0.1027>, rot=<0 0 0.2806 -0.9598>}
 ring_03_l {parent=ring_02_l, loc=<2.4152 0.114 0.0199>, rot=<0 0 0.1189 -0.9929>}
 pinky_metacarpal_l {parent=hand_l, loc=<1.7616 -0.9753 3.3814>, rot=<-0.0397 0.0814 -0.069 -0.9935>}
 pinky_01_l {parent=pinky_metacarpal_l, loc=<7.1709 -0.4831 -0.1094>, rot=<-0.0654 -0.0035 0.2242 -0.9723>}
 pinky_02_l {parent=pinky_01_l, loc=<2.8274 0.0893 -0.1043>, rot=<0.0037 0.0011 0.2854 -0.9584>}
 pinky_03_l {parent=pinky_02_l, loc=<2.2046 0.0307 -0.0078>, rot=<-0.0002 0.0098 0.1445 -0.9895>}
 lowerarm_twist_02_l {parent=lowerarm_l, loc=<9.6199 0 0>, rot=<0 0 0 -1>}
 lowerarm_twist_01_l {parent=lowerarm_l, loc=<17.8549 0 0>, rot=<0 0 0 -1>}
 cr_lowerarm_push_l {parent=lowerarm_l, loc=<0 0 0>, rot=<0 0 0 1>}
 upperarm_twist_01_l {parent=upperarm_l, loc=<7.9796 0.0843 -0.6195>, rot=<0 0 0 -1>}
 upperarm_twist_02_l {parent=upperarm_l, loc=<19.5059 0.0843 -0.6195>, rot=<0 0 0 -1>}
 upperarm_l_null {parent=clavicle_l, type=Null, loc=<18.2264 0.1342 -0.001>, rot=<0.0253 -0.2693 0.0013 -0.9627>}
 neck_01 {parent=spine_05, loc=<10.815 1.9196 0>, rot=<0 0 -0.1864 -0.9825>}
 neck_02 {parent=neck_01, loc=<4.7347 0.3045 0>, rot=<0 0 0.0363 -0.9993>}
 head {parent=neck_02, loc=<5.7632 0.5552 0>, rot=<0 0 0.115 -0.9934>}
 clavicle_r {parent=spine_05, loc=<2.5855 -0.2866 4.3777>, rot=<0.7289 0.0465 -0.6829 0.009>}
 upperarm_r {parent=clavicle_r, loc=<-18.2264 -0.1342 0.001>, rot=<0.0253 -0.2693 0.0013 -0.9627>}
 lowerarm_r {parent=upperarm_r, loc=<-29.0071 -0.1686 1.239>, rot=<0.0391 -0.0099 -0.2453 -0.9686>}
 hand_r {parent=lowerarm_r, loc=<-25.8607 -0.222 -1.0274>, rot=<-0.7193 -0.0988 0.1193 -0.6773>}
 hand_attach_r {parent=hand_r, loc=<-8.5 3 0>, rot=<0 0 -0.7071 -0.7071>}
 thumb_01_r {parent=hand_r, loc=<-1.3118 0.3191 2.459>, rot=<0.6546 -0.4042 0.0664 -0.6354>}
 thumb_02_r {parent=thumb_01_r, loc=<-5.9336 0.0448 -0.0401>, rot=<-0.0037 0.0213 0.298 -0.9543>}
 thumb_03_r {parent=thumb_02_r, loc=<-2.2223 -0.3682 0.0907>, rot=<0.0118 -0.013 0.0066 -0.9998>}
 index_metacarpal_r {parent=hand_r, loc=<-1.6535 -0.0183 2.96>, rot=<0.0158 -0.0124 -0.0173 -0.9996>}
 index_01_r {parent=index_metacarpal_r, loc=<-8.3008 -0.0357 0.0524>, rot=<-0.0044 0.0485 0.0924 -0.9945>}
 index_02_r {parent=index_01_r, loc=<-3.3587 0.3004 -0.0452>, rot=<-0.002 0.0027 0.3225 -0.9466>}
 index_03_r {parent=index_02_r, loc=<-2.5578 0.123 0.0252>, rot=<-0.002 0.0067 0.2558 -0.9667>}
 middle_metacarpal_r {parent=hand_r, loc=<-2.1856 0.1725 0.5984>, rot=<0.006 0.0168 -0.0329 -0.9993>}
 middle_01_r {parent=middle_metacarpal_r, loc=<-7.3275 0.2447 -0.0129>, rot=<-0.0208 0.0042 0.0858 -0.9961>}
 middle_02_r {parent=middle_01_r, loc=<-4.0123 -0.1401 -0.1175>, rot=<-0.0218 0.0258 0.3512 -0.9357>}
 middle_03_r {parent=middle_02_r, loc=<-2.6077 -0.0534 -0.1373>, rot=<0.0082 0.0188 0.1708 -0.9851>}
 ring_metacarpal_r {parent=hand_r, loc=<-2.2074 0.3625 -1.3638>, rot=<-0.0284 0.0722 -0.0458 -0.9959>}
 ring_01_r {parent=ring_metacarpal_r, loc=<-7.0295 0.6197 0.1446>, rot=<-0.0489 -0.0252 0.1892 -0.9804>}
 ring_02_r {parent=ring_01_r, loc=<-3.6174 -0.4647 0.1027>, rot=<0 0 0.2806 -0.9598>}
 ring_03_r {parent=ring_02_r, loc=<-2.4152 -0.114 -0.0199>, rot=<0 0 0.1189 -0.9929>}
 pinky_metacarpal_r {parent=hand_r, loc=<-1.7616 0.9753 -3.3814>, rot=<-0.0397 0.0814 -0.069 -0.9935>}
 pinky_01_r {parent=pinky_metacarpal_r, loc=<-7.1709 0.4831 0.1094>, rot=<-0.0654 -0.0035 0.2242 -0.9723>}
 pinky_02_r {parent=pinky_01_r, loc=<-2.8274 -0.0893 0.1043>, rot=<0.0037 0.0011 0.2854 -0.9584>}
 pinky_03_r {parent=pinky_02_r, loc=<-2.2046 -0.0307 0.0078>, rot=<-0.0002 0.0098 0.1445 -0.9895>}
 weapon_r {parent=hand_r, loc=<0.1974 3.4111 -0.3811>, rot=<0 0 0 1>}
 lowerarm_twist_02_r {parent=lowerarm_r, loc=<-9.6199 0 0>, rot=<0 0 0 -1>}
 lowerarm_twist_01_r {parent=lowerarm_r, loc=<-17.8549 0 0>, rot=<0 0 0 -1>}
 cr_lowerarm_push_r {parent=lowerarm_r, loc=<0 0 0>, rot=<0 0 0 1>}
 upperarm_twist_01_r {parent=upperarm_r, loc=<-7.9796 -0.0843 0.6195>, rot=<0 0 0 -1>}
 upperarm_twist_02_r {parent=upperarm_r, loc=<-19.5059 -0.0843 0.6195>, rot=<0 0 0 -1>}
 lowerarm_r_2 {parent=upperarm_r, loc=<-29.0071 -0.1686 1.239>, rot=<0.0391 -0.0099 -0.2453 -0.9686>}
 upperarm_r_null {parent=clavicle_r, type=Null, loc=<-18.2264 -0.1342 0.001>, rot=<0.0253 -0.2693 0.0013 -0.9627>}
 attach_cape {parent=spine_05, loc=<9.017 -11.714 0>, rot=<0.0616 -0.7044 -0.0616 -0.7044>}
 cape {parent=attach_cape, type=Null, loc=<0 -4.7064 5.0033>, rot=<0 0 0 1>}
 attach_backpack {parent=spine_05, loc=<-0.123 -13.473 0>, rot=<-0.0093 -0.707 0.0093 -0.707>}
 Backpack_BR {parent=attach_backpack, type=Null, loc=<0 -3.8146 14.2438>, rot=<0 0 0 1>}
 dyn_L_pauldron_bottom {parent=spine_05, loc=<4.5694 -1.3559 -21.3442>, rot=<0.0654 -0.8854 -0.0137 -0.46>}
 dyn_L_pauldron_top {parent=dyn_L_pauldron_bottom, loc=<-6.1168 -4.9546 13.3161>, rot=<-0.0096 0.2333 0.0399 -0.9715>}
 dyn_L_cardboardFlap_A {parent=dyn_L_pauldron_top, loc=<13.4271 15.2325 -2.8239>, rot=<0.5754 0.3895 -0.3576 -0.6239>}
 dyn_L_cardboardFlap_B {parent=dyn_L_pauldron_top, loc=<13.7089 15.2548 -10.2978>, rot=<0.6583 0.3283 -0.3419 -0.5848>}
 dyn_R_pauldron_bottom {parent=spine_05, loc=<4.5694 -1.3559 21.3442>, rot=<0.8854 0.0654 -0.46 0.0137>}
 dyn_R_pauldron_top {parent=dyn_R_pauldron_bottom, loc=<6.1168 4.9546 -13.3161>, rot=<-0.0096 0.2333 0.0399 -0.9715>}
 dyn_R_cardboardFlap_A {parent=dyn_R_pauldron_top, loc=<-13.4271 -15.2325 2.8239>, rot=<0.5754 0.3895 -0.3576 -0.6239>}
 dyn_R_cardboardFlap_B {parent=dyn_R_pauldron_top, loc=<-13.7089 -15.2548 10.2978>, rot=<0.6583 0.3283 -0.3419 -0.5848>}
 dyn_L_buttersArm {parent=spine_05, loc=<16.3044 5.6513 -9.0316>, rot=<-0.4281 -0.8851 -0.1099 0.146>}
 dyn_R_buttersArm {parent=spine_05, loc=<16.3046 5.6513 9.0316>, rot=<0.8851 -0.4281 0.146 0.1099>}
 dyn_L_chairScaler {parent=spine_05, loc=<3.3348 0 -2.6388>, rot=<0 0 0 1>}
 dyn_R_chairScaler {parent=spine_05, loc=<3.3347 0 2.6388>, rot=<0 0 1 0>}
 deform_clavicle_l {parent=spine_05, loc=<2.5856 -0.2866 -4.3777>, rot=<0.0465 -0.7289 -0.009 -0.6829>}
 deform_upperarm_l {parent=deform_clavicle_l, loc=<18.2264 0.1342 -0.0011>, rot=<0.0253 -0.2693 0.0013 -0.9627>}
 deform_lowerarm_l {parent=deform_upperarm_l, loc=<29.0071 0.1686 -1.239>, rot=<0.0391 -0.0099 -0.2453 -0.9686>}
 dyn_L_armScaler {parent=deform_lowerarm_l, loc=<4.2889 0.1674 -0.3568>, rot=<0 0 1 0>}
 deform_hand_l {parent=deform_lowerarm_l, loc=<25.8607 0.222 1.0274>, rot=<-0.7193 -0.0988 0.1193 -0.6773>}
 deform_thumb_01_l {parent=deform_hand_l, loc=<1.3118 -0.3191 -2.459>, rot=<0.6546 -0.4042 0.0664 -0.6354>}
 deform_thumb_02_l {parent=deform_thumb_01_l, loc=<5.9336 -0.0448 0.0401>, rot=<-0.0037 0.0213 0.298 -0.9543>}
 deform_thumb_03_l {parent=deform_thumb_02_l, loc=<2.2223 0.3682 -0.0907>, rot=<0.0118 -0.013 0.0066 -0.9998>}
 deform_index_metacarpal_l {parent=deform_hand_l, loc=<1.6535 0.0183 -2.96>, rot=<0.0158 -0.0124 -0.0173 -0.9996>}
 deform_index_01_l {parent=deform_index_metacarpal_l, loc=<8.3008 0.0357 -0.0524>, rot=<-0.0044 0.0485 0.0924 -0.9945>}
 deform_index_02_l {parent=deform_index_01_l, loc=<3.3587 -0.3004 0.0452>, rot=<-0.002 0.0027 0.3225 -0.9466>}
 deform_index_03_l {parent=deform_index_02_l, loc=<2.5578 -0.123 -0.0252>, rot=<-0.002 0.0067 0.2558 -0.9667>}
 deform_middle_metacarpal_l {parent=deform_hand_l, loc=<2.1856 -0.1725 -0.5984>, rot=<0.006 0.0168 -0.0329 -0.9993>}
 deform_middle_01_l {parent=deform_middle_metacarpal_l, loc=<7.3275 -0.2447 0.0129>, rot=<-0.0208 0.0042 0.0858 -0.9961>}
 deform_middle_02_l {parent=deform_middle_01_l, loc=<4.0123 0.1401 0.1175>, rot=<-0.0218 0.0258 0.3512 -0.9357>}
 deform_middle_03_l {parent=deform_middle_02_l, loc=<2.6077 0.0534 0.1373>, rot=<0.0082 0.0188 0.1708 -0.9851>}
 deform_ring_metacarpal_l {parent=deform_hand_l, loc=<2.2074 -0.3625 1.3638>, rot=<-0.0284 0.0722 -0.0458 -0.9959>}
 deform_ring_01_l {parent=deform_ring_metacarpal_l, loc=<7.0295 -0.6197 -0.1446>, rot=<-0.0489 -0.0252 0.1892 -0.9804>}
 deform_ring_02_l {parent=deform_ring_01_l, loc=<3.6174 0.4647 -0.1027>, rot=<0 0 0.2806 -0.9598>}
 deform_ring_03_l {parent=deform_ring_02_l, loc=<2.4152 0.114 0.0199>, rot=<0 0 0.1189 -0.9929>}
 deform_pinky_metacarpal_l {parent=deform_hand_l, loc=<1.7616 -0.9753 3.3814>, rot=<-0.0397 0.0814 -0.069 -0.9935>}
 deform_pinky_01_l {parent=deform_pinky_metacarpal_l, loc=<7.1709 -0.4831 -0.1094>, rot=<-0.0654 -0.0035 0.2242 -0.9723>}
 deform_pinky_02_l {parent=deform_pinky_01_l, loc=<2.8274 0.0893 -0.1043>, rot=<0.0037 0.0011 0.2854 -0.9584>}
 deform_pinky_03_l {parent=deform_pinky_02_l, loc=<2.2046 0.0307 -0.0078>, rot=<-0.0002 0.0098 0.1445 -0.9895>}
 deform_lowerarm_gauntles_l {parent=deform_upperarm_l, loc=<29.0071 0.1686 -1.239>, rot=<0.0391 -0.0099 -0.2453 -0.9686>}
 deform_clavicle_r {parent=spine_05, loc=<2.5853 -0.2866 4.3777>, rot=<0.7289 0.0465 -0.6829 0.009>}
 deform_upperarm_r {parent=deform_clavicle_r, loc=<-18.2264 -0.1342 0.0013>, rot=<0.0253 -0.2693 0.0013 -0.9627>}
 deform_lowerarm_r {parent=deform_upperarm_r, loc=<-29.0067 -0.1685 1.2385>, rot=<0.0391 -0.0099 -0.2453 -0.9686>}
 dyn_R_armScaler {parent=deform_lowerarm_r, loc=<-4.2891 -0.1674 0.357>, rot=<0 0 1 0>}
 deform_hand_r {parent=deform_lowerarm_r, loc=<-25.8609 -0.2219 -1.0273>, rot=<-0.7193 -0.0988 0.1193 -0.6773>}
 deform_thumb_01_r {parent=deform_hand_r, loc=<-1.312 0.3194 2.459>, rot=<0.6546 -0.4042 0.0664 -0.6354>}
 deform_thumb_02_r {parent=deform_thumb_01_r, loc=<-5.9333 0.0446 -0.0402>, rot=<-0.0037 0.0213 0.298 -0.9543>}
 deform_thumb_03_r {parent=deform_thumb_02_r, loc=<-2.2223 -0.3682 0.0908>, rot=<0.0118 -0.013 0.0066 -0.9998>}
 deform_index_metacarpal_r {parent=deform_hand_r, loc=<-1.6537 -0.018 2.96>, rot=<0.0158 -0.0124 -0.0173 -0.9996>}
 deform_index_01_r {parent=deform_index_metacarpal_r, loc=<-8.3005 -0.036 0.0526>, rot=<-0.0044 0.0485 0.0924 -0.9945>}
 deform_index_02_r {parent=deform_index_01_r, loc=<-3.3587 0.3004 -0.0453>, rot=<-0.002 0.0027 0.3225 -0.9466>}
 deform_index_03_r {parent=deform_index_02_r, loc=<-2.5578 0.123 0.0253>, rot=<-0.002 0.0067 0.2558 -0.9667>}
 deform_middle_metacarpal_r {parent=deform_hand_r, loc=<-2.1855 0.1725 0.5984>, rot=<0.006 0.0168 -0.0329 -0.9993>}
 deform_middle_01_r {parent=deform_middle_metacarpal_r, loc=<-7.3275 0.2448 -0.0129>, rot=<-0.0208 0.0042 0.0858 -0.9961>}
 deform_middle_02_r {parent=deform_middle_01_r, loc=<-4.0123 -0.1402 -0.1175>, rot=<-0.0218 0.0258 0.3512 -0.9357>}
 deform_middle_03_r {parent=deform_middle_02_r, loc=<-2.6077 -0.0534 -0.1373>, rot=<0.0082 0.0188 0.1708 -0.9851>}
 deform_ring_metacarpal_r {parent=deform_hand_r, loc=<-2.2073 0.3624 -1.3638>, rot=<-0.0284 0.0722 -0.0458 -0.9959>}
 deform_ring_01_r {parent=deform_ring_metacarpal_r, loc=<-7.0294 0.6197 0.1446>, rot=<-0.0489 -0.0252 0.1892 -0.9804>}
 deform_ring_02_r {parent=deform_ring_01_r, loc=<-3.6175 -0.4647 0.1026>, rot=<0 0 0.2806 -0.9598>}
 deform_ring_03_r {parent=deform_ring_02_r, loc=<-2.4152 -0.1141 -0.0199>, rot=<0 0 0.1189 -0.9929>}
 deform_pinky_metacarpal_r {parent=deform_hand_r, loc=<-1.7615 0.9753 -3.3813>, rot=<-0.0397 0.0814 -0.069 -0.9935>}
 deform_pinky_01_r {parent=deform_pinky_metacarpal_r, loc=<-7.171 0.4831 0.1093>, rot=<-0.0654 -0.0035 0.2242 -0.9723>}
 deform_pinky_02_r {parent=deform_pinky_01_r, loc=<-2.8274 -0.0892 0.1043>, rot=<0.0037 0.0011 0.2854 -0.9584>}
 deform_pinky_03_r {parent=deform_pinky_02_r, loc=<-2.2046 -0.0307 0.0078>, rot=<-0.0002 0.0098 0.1445 -0.9895>}
 deform_lowerarm_gauntles_r {parent=deform_upperarm_r, loc=<-29.0067 -0.1685 1.2385>, rot=<0.0391 -0.0099 -0.2453 -0.9686>}
 base_neck_seat {parent=spine_05, loc=<27.3257 -9.6374 -0.0196>, rot=<0.0401 -0.9992 0 0>}
 dyn_neck_seat {parent=base_neck_seat, loc=<0 0 0>, rot=<0 0 0 1>}
 neck_01_null {parent=spine_05, type=Null, loc=<10.815 1.9196 0>, rot=<0 0 -0.1864 -0.9825>}
 dyn_neck_seat1 {parent=spine_05, loc=<27.3257 -9.6374 -0.0196>, rot=<0.0401 -0.9992 0 0>}
 dyn_backFlap_1 {parent=spine_01, loc=<5.3796 -9.085 -0.0093>, rot=<0.1697 -0.9855 0.0001 -0.0008>}
 dyn_backFlap_2 {parent=dyn_backFlap_1, loc=<11.4282 -0.5321 0>, rot=<0 0 0 1>}
 dyn_L_frontFlap_mstr {parent=spine_01, loc=<3.8142 7.3973 -4.1927>, rot=<-0.137 -0.9906 -0.0005 -0.0029>}
 dyn_L_frontFlap_1 {parent=dyn_L_frontFlap_mstr, loc=<0 0 0>, rot=<0 0 0 1>}
 dyn_L_frontFlap_2 {parent=dyn_L_frontFlap_1, loc=<6.8612 0.2293 0>, rot=<0 0 0 1>}
 dyn_L_frontFlap_3 {parent=dyn_L_frontFlap_2, loc=<5.581 0 0>, rot=<0 0 0 1>}
 dyn_R_frontFlap_mstr {parent=spine_01, loc=<3.8146 7.3973 4.1927>, rot=<0.9906 -0.137 -0.0029 0.0005>}
 dyn_R_frontFlap_1 {parent=dyn_R_frontFlap_mstr, loc=<0 0 0>, rot=<0 0 0 1>}
 dyn_R_frontFlap_2 {parent=dyn_R_frontFlap_1, loc=<-6.8616 -0.2292 0>, rot=<0 0 0 1>}
 dyn_R_frontFlap_3 {parent=dyn_R_frontFlap_2, loc=<-5.581 0 0>, rot=<0 0 0 1>}
 thigh_l {parent=pelvis, loc=<-1.8716 -0.0109 -11.9936>, rot=<0.0012 0.0572 0.0155 -0.9982>}
 calf_l {parent=thigh_l, loc=<-41.9178 -0.2246 0.0926>, rot=<0.0006 -0.0161 -0.0391 -0.9991>}
 foot_l {parent=calf_l, loc=<-38.7057 1.2438 0.2452>, rot=<0.0975 -0.0474 0.0317 -0.9936>}
 ball_l {parent=foot_l, loc=<-6.1614 13.1469 0.6807>, rot=<0.0001 -0.0001 -0.7063 -0.7079>}
 dyn_L_ductTapeFoot_back {parent=foot_l, loc=<-1.988 -9.5176 -2.2779>, rot=<0.4721 -0.8739 -0.116 0.0017>}
 dyn_L_ballPivot {parent=foot_l, loc=<-7.8752 4.3061 0.6822>, rot=<0.0001 -0.0001 -0.7063 -0.7079>}
 dyn_L_ductTapeFoot_front {parent=dyn_L_ballPivot, loc=<13.9109 -6.2751 2.6957>, rot=<0.9449 0.3098 0.1056 0.0032>}
 calf_twist_02_l {parent=calf_l, loc=<-15.9268 -0.0967 0.0397>, rot=<0.0237 0.0095 -0.0035 -0.9997>}
 calf_twist_01_l {parent=calf_l, loc=<-32.9335 -0.0967 0.0397>, rot=<0.0237 0.0095 -0.0035 -0.9997>}
 dyn_L_upperBack_scaler {parent=calf_l, loc=<-6.3482 -1.8178 0.124>, rot=<0 0 0.4125 -0.9109>}
 dyn_L_lowerFront_scaler {parent=calf_l, loc=<-29.8219 8.7136 -0.2>, rot=<0 0 0.9419 0.336>}
 dyn_L_lowerBack_scaler {parent=calf_l, loc=<-30.188 0.9946 -0.2>, rot=<0 0 0.9643 -0.2648>}
 thigh_twist_01_l {parent=thigh_l, loc=<-21.6303 0 0>, rot=<0.0013 0 0 -1>}
 calf_Boots_l {parent=thigh_twist_01_l, loc=<-20.2875 -0.2249 0.092>, rot=<-0.0006 -0.016 -0.0391 -0.9991>}
 thigh_r {parent=pelvis, loc=<-1.8716 -0.0109 11.9936>, rot=<0.0572 -0.0012 0.9982 0.0155>}
 calf_r {parent=thigh_r, loc=<41.9178 0.2246 -0.0926>, rot=<0.0006 -0.0161 -0.0391 -0.9991>}
 foot_r {parent=calf_r, loc=<38.7057 -1.2438 -0.2452>, rot=<0.0975 -0.0474 0.0317 -0.9936>}
 ball_r {parent=foot_r, loc=<6.1614 -13.1469 -0.6807>, rot=<0.0001 -0.0001 -0.7063 -0.7079>}
 dyn_R_ductTapeFoot_back {parent=foot_r, loc=<1.988 9.5176 2.2779>, rot=<0.4721 -0.8739 -0.116 0.0017>}
 dyn_R_ballPivot {parent=foot_r, loc=<7.8752 -4.3061 -0.6822>, rot=<0.0001 -0.0001 -0.7063 -0.7079>}
 dyn_R_ductTapeFoot_front {parent=dyn_R_ballPivot, loc=<-13.9109 6.2751 -2.6958>, rot=<0.9449 0.3098 0.1056 0.0032>}
 calf_twist_02_r {parent=calf_r, loc=<15.9268 0.0967 -0.0397>, rot=<0.0237 0.0095 -0.0035 -0.9997>}
 calf_twist_01_r {parent=calf_r, loc=<32.9335 0.0967 -0.0397>, rot=<0.0237 0.0095 -0.0035 -0.9997>}
 dyn_R_upperBack_scaler {parent=calf_r, loc=<6.3483 1.8178 -0.124>, rot=<0 0 0.4125 -0.9109>}
 dyn_R_lowerFront_scaler {parent=calf_r, loc=<29.8218 -8.7136 0.2>, rot=<0 0 0.9419 0.336>}
 dyn_R_lowerBack_scaler {parent=calf_r, loc=<30.188 -0.9946 0.2>, rot=<0 0 0.9643 -0.2648>}
 thigh_twist_01_r {parent=thigh_r, loc=<21.6303 0 0>, rot=<0.0013 0 0 -1>}
 calf_Boots_r {parent=thigh_twist_01_r, loc=<20.2875 0.2249 -0.092>, rot=<-0.0006 -0.016 -0.0391 -0.9991>}
 attach {parent=root, loc=<0 0 -2.1501>, rot=<0 0 -0.1846 -0.9828>}
 ik_hand_root {parent=root, loc=<0 0 0>, rot=<0 0 0 1>}
 ik_hand_gun {parent=ik_hand_root, loc=<-64.462 4.8638 101.1961>, rot=<0.6084 0.1793 0.4613 -0.6205>}
 ik_hand_r {parent=ik_hand_gun, loc=<0 0 0>, rot=<0 0 0 1>}
 ik_hand_l {parent=ik_hand_gun, loc=<65.7708 101.9258 43.6659>, rot=<-0.5102 -0.7906 -0.3387 0>}
 ik_foot_root {parent=root, loc=<-0.0052 0.01 -2.1501>, rot=<-0.0004 -0.0002 -0.1846 -0.9828>}
 ik_foot_l {parent=ik_foot_root, loc=<21.7353 17.5718 9.515>, rot=<-0.0307 0.7047 -0.0362 -0.7079>}
 ik_foot_r {parent=ik_foot_root, loc=<-14.3826 -24.2258 9.5749>, rot=<0.6651 0.2483 -0.6567 0.2545>}
 root_fp {parent=root, loc=<0 0 -2.1501>, rot=<0 0 -0.1846 -0.9828>}
 camera_root {parent=root_fp, loc=<0 0 0>, rot=<0 0 0 1>}
 camera {parent=camera_root, loc=<0 0 0>, rot=<0 0 0 1>}
 attach_fp {parent=root_fp, loc=<0 0 0>, rot=<0 0 0 1>}
 attach_emote_01 {parent=root, loc=<0 0 -2.1501>, rot=<0 0 -0.1846 -0.9828>}
 attach_emote_02 {parent=root, loc=<0 0 -2.1501>, rot=<0 0 -0.1846 -0.9828>}
 attach_emote_03 {parent=root, loc=<0 0 -2.1501>, rot=<0 0 -0.1846 -0.9828>}
 warp_point_01 {parent=root, loc=<0 0 -2.1501>, rot=<0 0 -0.1846 -0.9828>}
 warp_point_02 {parent=root, loc=<0 0 -2.1501>, rot=<0 0 -0.1846 -0.9828>}
 warp_point_03 {parent=root, loc=<0 0 -2.1501>, rot=<0 0 -0.1846 -0.9828>}
]
variables [
 L Elbow Distance to Head {type=double, default=0}
 L armBend alpha {type=double, default=0}
 L calfBent alpha {type=double, default=0}
 L footBentBack alpha {type=double, default=0}
 L footBentUp alpha {type=double, default=0}
 Pub_CelebrationEmote {type=bool, default=False}
 Pub_IsStickOfTruth {type=bool, default=False}
 Pub_ParachuteOpened {type=bool, default=False}
 Pub_baseballsliding {type=bool, default=False}
 Pub_kneesliding {type=bool, default=False}
 R Elbow Distance to Head {type=double, default=0}
 R armBend alpha {type=double, default=0}
 R calfBent alpha {type=double, default=0}
 R footBentBack alpha {type=double, default=0}
 R footBentUp alpha {type=double, default=0}
]
