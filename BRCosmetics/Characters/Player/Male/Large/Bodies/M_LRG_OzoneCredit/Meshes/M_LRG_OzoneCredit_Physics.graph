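# Physics asset for the large male OzoneCredit player body cosmetic, set up for its left and right front clothing flaps.
# Each flap is a three-box chain constrained to a small master capsule, with identical angular limits that favor one swing axis.
# Capsules on the pelvis, lower spine and both thighs have no constraints, apparently serving as collision shapes for the flaps.
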PhysicsAsset /BRCosmetics/Characters/Player/Male/Large/Bodies/M_LRG_OzoneCredit/Meshes/M_LRG_OzoneCredit_Physics
bodies {
 dyn_L_frontFlap_1 [{shapeType=Box, extentX=10, extentY=4, extentZ=13.3801, shapeName=dyn_L_frontFlap_1_box, center=<3 0 0>}]
 dyn_L_frontFlap_2 [{shapeType=Box, extentX=10, extentY=4.5555, extentZ=13.3801, shapeName=dyn_L_frontFlap_2_box, center=<3 0 0>}]
 dyn_L_frontFlap_3 [{shapeType=Box, extentX=10, extentY=5.0662, extentZ=13.3801, shapeName=dyn_L_frontFlap_3_box, center=<3 0 0>}]
 dyn_L_frontFlap_mstr [{shapeType=Capsule, length=1, radius=1}]
 dyn_R_frontFlap_1 [{shapeType=Box, extentX=10, extentY=4, extentZ=13.3801, shapeName=dyn_R_frontFlap_1_box, center=<-3 0 0>}]
 dyn_R_frontFlap_2 [{shapeType=Box, extentX=10, extentY=4.5555, extentZ=13.3801, shapeName=dyn_R_frontFlap_2_box, center=<-3 0 0>}]
 dyn_R_frontFlap_3 [{shapeType=Box, extentX=10, extentY=5.0662, extentZ=13.3801, shapeName=dyn_R_frontFlap_3_box, center=<-3 0 0>}]
 dyn_R_frontFlap_mstr [{shapeType=Capsule, length=1, radius=1}]
 pelvis [{shapeType=Capsule, length=19, radius=8.8, center=<0 -2 0>}]
 spine_01 [{shapeType=Capsule, length=19, radius=8.5, center=<0 -3 0>}]
 thigh_l [{shapeType=Capsule, length=20, radius=7.5, shapeName=thigh_l_capsule, center=<-11 0 0>, rotation=<rot 90 0 0>}]
 thigh_r [{shapeType=Capsule, length=20, radius=7.5, shapeName=thigh_l_capsule, center=<11 0 0>, rotation=<rot 90 0 0>}]
}
constraints [
 {bone1Name=dyn_L_frontFlap_1, bone2Name=dyn_L_frontFlap_mstr, swing1LimitDegrees=45, swing2LimitDegrees=10, twistLimitDegrees=20}
 {bone1Name=dyn_L_frontFlap_2, bone2Name=dyn_L_frontFlap_1, swing1LimitDegrees=45, swing2LimitDegrees=10, twistLimitDegrees=20}
 {bone1Name=dyn_L_frontFlap_3, bone2Name=dyn_L_frontFlap_2, swing1LimitDegrees=45, swing2LimitDegrees=10, twistLimitDegrees=20}
 {bone1Name=dyn_R_frontFlap_1, bone2Name=dyn_R_frontFlap_mstr, swing1LimitDegrees=45, swing2LimitDegrees=10, twistLimitDegrees=20}
 {bone1Name=dyn_R_frontFlap_2, bone2Name=dyn_R_frontFlap_1, swing1LimitDegrees=45, swing2LimitDegrees=10, twistLimitDegrees=20}
 {bone1Name=dyn_R_frontFlap_3, bone2Name=dyn_R_frontFlap_2, swing1LimitDegrees=45, swing2LimitDegrees=10, twistLimitDegrees=20}
]
