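# Animation Blueprint for the M_LRG_OzoneCredit player body, layering secondary motion over the pose copied from its attached parent mesh.
# A Control Rig receives sliding, parachute, emote and stick-of-truth flags, then constraints, RigidBody and AnimDynamics drive back flaps, duct-tape feet and other dynamic bones.
# Later nodes constrain boots and gauntlets and scale the gauntlet bones.
# Events set front-end tuning, skydiving gravity, arm visibility and weapon-type variables.
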AnimBlueprint /BRCosmetics/Characters/Player/Male/Large/Bodies/M_LRG_OzoneCredit/Meshes/M_LRG_OzoneCredit_AnimBP
target_skeleton=&/M_LRG_OzoneCredit_Skeleton
graphs [
 AnimGraph {nodes [
   n1 = Control Rig!(
    Source=Layered blend per bone(BasePose=Copy Pose From Mesh() {bUseAttachedParent=true}) {layerSetup [{branchFilters [{blendDepth=1, boneName=neck_02}]}, {branchFilters []}]}
    Alpha
    Pub_kneesliding:bool=Get bIsKneeSliding()
    Pub_baseballsliding:bool=Get bIsBaseballSliding()
    Pub_CelebrationEmote:bool=Get IsCelebrationEmote() -> (IsCelebrationEmote:bool)
    Pub_ParachuteOpened:bool=Get bIsParachuteOpened()
    Pub_IsStickOfTruth:bool=Get IsStickOfTruth() -> (IsStickOfTruth:bool)
   ) -> (Pose) {controlRigAssetReference {blueprintRigClass=&/M_LRG_OzoneCredit_CtrlRig.M_LRG_OzoneCredit_CtrlRig_C}}
   n2 = "Constraint - dyn_L_ballPivot (P)"(ComponentPose=Local To Component(LocalPose=@n1), Alpha=1, ConstraintWeights_0=1) {boneToModify {boneName=dyn_L_ballPivot}, constraintSetup [{targetBone {boneName=ball_l}}], constraintWeights [1]}
   n3 = "Constraint - dyn_R_ballPivot (P)"(ComponentPose=@n2, Alpha=1, ConstraintWeights_0=1) {boneToModify {boneName=dyn_R_ballPivot}, constraintSetup [{targetBone {boneName=ball_r}}], constraintWeights [1]}
   n4 = RigidBody!(ComponentPose=@n3, Alpha, OverrideWorldGravity:Vector="0,0,-980") -> (Pose) {
    bOverrideWorldGravity=true
    overridePhysicsAsset=&/M_LRG_OzoneCredit_Physics
    simulationSpace=BaseBoneSpace
    baseBoneRef {boneName=pelvis}
    overrideWorldGravity {z=-980}
    simSpaceSettings {
     dampingAlpha=0
     maxAngularAcceleration=50
     maxAngularVelocity=7
     maxLinearAcceleration=4000
     maxLinearVelocity=800
     velocityScaleZ=0.5
     worldAlpha=1
     externalLinearDragV=<1.5 1.5 1.5>
    }
   }
   n5 = Anim Dynamics - Bone: dyn_backFlap_1!(ComponentPose=@n4, Alpha=1, GravityOverride:Vector="0,0,-980", AngularSpringConstant:float=20000) -> (Pose) {
    angularDampingOverride=1
    angularSpringConstant=20000
    bAngularSpring=true
    bOverrideAngularDamping=true
    bUseGravityOverride=true
    gravityScale=0.35
    simulationSpace=RootRelative
    boundBone {boneName=dyn_backFlap_1}
    componentAppliedLinearAccClamp=<4000 4000 4000>
    componentLinearAccScale=<1 1 1>
    componentLinearVelScale=<1 1 1>
    gravityOverride {z=-980}
    physicsBodyDefinitions [{
      boundBone {boneName=dyn_backFlap_1}
      boxExtents=<29 29 29>
      constraintSetup {angularLimitsMax=<0 10 20>, angularLimitsMin=<0 -10 -80>, angularTarget=<1 0 0>}
      localJointOffset=<2 0 0>
     }]
    simSpaceSettings {maxAngularAcceleration=40, maxAngularVelocity=8, simSpaceAngularAlpha=1}
   }
   n6 = Anim Dynamics - Bone: dyn_backFlap_2!(ComponentPose=@n5, Alpha=1, GravityOverride:Vector="0,0,-980", AngularSpringConstant:float=20000) -> (Pose) {
    angularDampingOverride=1
    angularSpringConstant=20000
    bAngularSpring=true
    bOverrideAngularDamping=true
    bUseGravityOverride=true
    gravityScale=0.35
    simulationSpace=RootRelative
    boundBone {boneName=dyn_backFlap_2}
    componentAppliedLinearAccClamp=<4000 4000 4000>
    componentLinearAccScale=<1 1 1>
    componentLinearVelScale=<1 1 1>
    gravityOverride {z=-980}
    physicsBodyDefinitions [{
      boundBone {boneName=dyn_backFlap_2}
      boxExtents=<29 29 29>
      constraintSetup {angularLimitsMax=<0 10 30>, angularLimitsMin=<0 -10 -80>, angularTarget=<1 0 0>}
      localJointOffset=<2 0 0>
     }]
    simSpaceSettings {maxAngularAcceleration=40, maxAngularVelocity=8, simSpaceAngularAlpha=1}
   }
   n7 = Anim Dynamics - Bone: dyn_L_ductTapeFoot_front!(ComponentPose=@n6, Alpha=1, GravityOverride:Vector="0,0,-980") -> (Pose) {
    angularDampingOverride=0.9999
    angularSpringConstant=5000
    bAngularSpring=true
    bOverrideAngularDamping=true
    bUseGravityOverride=true
    gravityScale=0.35
    simulationSpace=BoneRelative
    boundBone {boneName=dyn_L_ductTapeFoot_front}
    componentAppliedLinearAccClamp=<4000 4000 4000>
    componentLinearAccScale=<1 1 1>
    componentLinearVelScale=<1 1 1>
    gravityOverride {z=-980}
    physicsBodyDefinitions [{
      boundBone {boneName=dyn_L_ductTapeFoot_front}
      boxExtents=<20 20 20>
      constraintSetup {angularLimitsMax=<0 0 60>, angularLimitsMin=<0 0 -3>, angularTarget=<1 0 0>}
      localJointOffset=<2 0 0>
     }]
    relativeSpaceBone {boneName=calf_l}
    simSpaceSettings {maxAngularAcceleration=40, maxAngularVelocity=8, simSpaceAngularAlpha=1}
   }
   n8 = Anim Dynamics - Bone: dyn_L_ductTapeFoot_back!(ComponentPose=@n7, Alpha=1, GravityOverride:Vector="0,0,-980") -> (Pose) {
    angularDampingOverride=0.9999
    angularSpringConstant=5000
    bAngularSpring=true
    bOverrideAngularDamping=true
    bUseGravityOverride=true
    gravityScale=0.35
    simulationSpace=BoneRelative
    boundBone {boneName=dyn_L_ductTapeFoot_back}
    componentAppliedLinearAccClamp=<4000 4000 4000>
    componentLinearAccScale=<1 1 1>
    componentLinearVelScale=<1 1 1>
    gravityOverride {z=-980}
    physicsBodyDefinitions [{
      boundBone {boneName=dyn_L_ductTapeFoot_back}
      boxExtents=<20 20 20>
      constraintSetup {angularLimitsMax=<0 0 3>, angularLimitsMin=<0 0 -60>, angularTarget=<1 0 0>}
      localJointOffset=<2 0 0>
     }]
    relativeSpaceBone {boneName=calf_l}
    simSpaceSettings {maxAngularAcceleration=40, maxAngularVelocity=8, simSpaceAngularAlpha=1}
   }
   n9 = Anim Dynamics - Bone: dyn_R_ductTapeFoot_front!(ComponentPose=@n8, Alpha=1, GravityOverride:Vector="0,0,-980") -> (Pose) {
    angularDampingOverride=0.9999
    angularSpringConstant=5000
    bAngularSpring=true
    bOverrideAngularDamping=true
    bUseGravityOverride=true
    gravityScale=0.35
    simulationSpace=BoneRelative
    boundBone {boneName=dyn_R_ductTapeFoot_front}
    componentAppliedLinearAccClamp=<4000 4000 4000>
    componentLinearAccScale=<1 1 1>
    componentLinearVelScale=<1 1 1>
    gravityOverride {z=-980}
    physicsBodyDefinitions [{
      boundBone {boneName=dyn_R_ductTapeFoot_front}
      boxExtents=<20 20 20>
      constraintSetup {angularLimitsMax=<0 0 60>, angularLimitsMin=<0 0 -3>, angularTarget=<1 0 0>}
      localJointOffset=<-2 0 0>
     }]
    relativeSpaceBone {boneName=calf_r}
    simSpaceSettings {maxAngularAcceleration=40, maxAngularVelocity=8, simSpaceAngularAlpha=1}
   }
   n10 = Anim Dynamics - Bone: dyn_R_ductTapeFoot_back!(ComponentPose=@n9, Alpha=1, GravityOverride:Vector="0,0,-980") -> (Pose) {
    angularDampingOverride=0.9999
    angularSpringConstant=5000
    bAngularSpring=true
    bOverrideAngularDamping=true
    bUseGravityOverride=true
    gravityScale=0.35
    simulationSpace=BoneRelative
    boundBone {boneName=dyn_R_ductTapeFoot_back}
    componentAppliedLinearAccClamp=<4000 4000 4000>
    componentLinearAccScale=<1 1 1>
    componentLinearVelScale=<1 1 1>
    gravityOverride {z=-980}
    physicsBodyDefinitions [{
      boundBone {boneName=dyn_R_ductTapeFoot_back}
      boxExtents=<20 20 20>
      constraintSetup {angularLimitsMax=<0 0 3>, angularLimitsMin=<0 0 -60>, angularTarget=<1 0 0>}
      localJointOffset=<-2 0 0>
     }]
    relativeSpaceBone {boneName=calf_r}
    simSpaceSettings {maxAngularAcceleration=40, maxAngularVelocity=8, simSpaceAngularAlpha=1}
   }
   n11 = "Constraint - deform_lowerarm_gauntles_l (P)"!(ComponentPose=@n10, bAlphaBoolEnabled:bool=True, ConstraintWeights_0=1) -> (Pose) {
    alphaInputType=Bool
    boneToModify {boneName=deform_lowerarm_gauntles_l}
    constraintSetup [{targetBone {boneName=lowerarm_l}}]
    constraintWeights [1]
   }
   n12 = "Constraint - deform_lowerarm_gauntles_r (P)"!(ComponentPose=@n11, bAlphaBoolEnabled:bool=True, ConstraintWeights_0=1) -> (Pose) {
    alphaInputType=Bool
    boneToModify {boneName=deform_lowerarm_gauntles_r}
    constraintSetup [{targetBone {boneName=lowerarm_r}}]
    constraintWeights [1]
   }
   n13 = "Constraint - calf_Boots_l (P)"(ComponentPose=@n12, Alpha=1, ConstraintWeights_0=1) {boneToModify {boneName=calf_Boots_l}, constraintSetup [{targetBone {boneName=calf_l}}], constraintWeights [1]}
   n14 = "Constraint - calf_Boots_r (P)"(ComponentPose=@n13, Alpha=1, ConstraintWeights_0=1) {boneToModify {boneName=calf_Boots_r}, constraintSetup [{targetBone {boneName=calf_r}}], constraintWeights [1]}
   n15 = "Transform (Modify) Bone - Bone: deform_lowerarm_gauntles_l"!(ComponentPose=@n14, bAlphaBoolEnabled:bool=True, Translation="0,0,0", Rotation="0,0,0", Scale="0.5,0.5,0.5") -> (Pose) {
    alphaInputType=Bool
    rotationMode=BMM_Additive
    scaleMode=BMM_Additive
    scaleSpace=BCS_BoneSpace
    translationMode=BMM_Additive
    translationSpace=BCS_BoneSpace
    boneToModify {boneName=deform_lowerarm_gauntles_l}
    scale=<0.5 0.5 0.5>
   }
   n16 = "Transform (Modify) Bone - Bone: deform_lowerarm_gauntles_r"!(ComponentPose=@n15, bAlphaBoolEnabled:bool=True, Translation="0,0,0", Rotation="0,0,0", Scale="0.5,0.5,0.5") -> (Pose) {
    alphaInputType=Bool
    rotationMode=BMM_Additive
    scaleMode=BMM_Additive
    scaleSpace=BCS_BoneSpace
    translationMode=BMM_Additive
    translationSpace=BCS_BoneSpace
    boneToModify {boneName=deform_lowerarm_gauntles_r}
    scale=<0.5 0.5 0.5>
   }
   Output Pose(Result=Component To Local(ComponentPose=@n16)) {layerGroup=None}
  ]}
 EventGraph {nodes [
   Event BlueprintUpdateAnimation()
   n1 = GetAnimInstance(self=GetOwnerSkeletalMeshComponent())
   > Is Valid(InputObject=@n1)
   > Set IsCelebrationEmote(IsCelebrationEmote:bool=Montage_IsPlaying(self=@n1, Montage=/Game/Animation/Game/MainPlayer/Emotes/Emote_Celebration_M)) -> (Output_Get:bool)
   n2 = Get bIsInFrontEnd()
   > Set AlphaDyn(AlphaDyn:float=SelectFloat(A=0.3, B=1, bPickA=@n2)) -> (Output_Get:float)
   > Is Valid(InputObject=@n1)
   n3 = Get bIsInFrontEnd()
   n4 = Get bIsPlayingEmote()
   > Set IsStickOfTruth(IsStickOfTruth:bool=AND Boolean!(
     A=Montage_IsActive!(self:Anim Instance Object Reference=@n1, Montage:Anim Montage Object Reference) -> (ReturnValue:bool)
     B=NOT Boolean(A=@n3)
     C:bool=NOT Boolean(A=@n4)
     D:bool=Get isGunOut() -> (isGunOut:bool)
    ) -> (ReturnValue)) -> (Output_Get:bool)
   Event OnCharacterCustomizationLoaded()
   > n5 = Sequence!(execute) -> (then_0, then_1, then_2:Exec)
   Set GravityOverride(execute=[@n8.then, @n5.then_0], GravityOverride="0,0,-980")
   Set SpringConstant(execute=@n5.then_1, SpringConstant:double=SelectFloat(A=14000, B=20000, bPickA=@n2)) -> (Output_Get:double)
   n6 = Event OnLeftArmVisible() {is=title K2Node_Event}
   Set IsWearingGauntlets(execute=[@n6.then, @n12.then], IsWearingGauntlets:bool=false) -> (Output_Get:bool)
   n7 = Event OnRightArmHidden() {is=title K2Node_Event}
   Set IsWearingGauntlets(execute=[@n7.then, @n9.then], IsWearingGauntlets=true)
   n8 = Event OnEndSkydiving()
   Event OnBeginSkydiving()
   > Set GravityOverride(GravityOverride="0,0,600")
   n9 = Event OnLeftArmHidden() {is=title K2Node_Event}
   Event OnWeaponChanged()
   > n10 = Cast To FortPlayerAnimInstance(Object=GetAnimInstance(self=GetOwnerSkeletalMeshComponent()))
   > Set isGunOut(isGunOut:bool=false) -> (Output_Get:bool)
   > n11 = Switch on EFortWeaponCoreAnimation!(execute, Selection:EFortWeaponCoreAnimation Enum=Get WeaponCoreAnim(self=@n10)) -> (
    Melee:Exec
    Pistol:Exec
    Shotgun:Exec
    PaperBlueprint:Exec
    Rifle:Exec
    MeleeOneHand:Exec
    MachinePistol:Exec
    RocketLauncher:Exec
    GrenadeLauncher:Exec
    GoingCommando:Exec
    AssaultRifle:Exec
    TacticalShotgun:Exec
    SniperRifle:Exec
    TrapPlacement:Exec
    ShoulderLauncher:Exec
    AbilityDecoTool:Exec
    Crossbow:Exec
    C4:Exec
    RemoteControl:Exec
    DualWield:Exec
    AR_BullPup:Exec
    AR_ForwardGrip:Exec
    MedPackPaddles:Exec
    SMG_P90:Exec
    AR_DrumGun:Exec
    Consumable_Small:Exec
    Consumable_Large:Exec
    Balloon:Exec
    MountedTurret:Exec
    CreativeTool:Exec
    ExplosiveBow:Exec
    AshtonIndigo:Exec
    AshtonChicago:Exec
    MeleeDualWield:Exec
    Unarmed:Exec
   )
   Set isGunOut(execute=[@n11.Melee, @n11.MeleeDualWield, @n11.MeleeOneHand, @n11.Unarmed], isGunOut=true)
   n12 = Event OnRightArmVisible() {is=title K2Node_Event}
  ]}
]
variables [
 AlphaDyn {type=number, default=1}
 "deform_lowerarm_r : Parent" {type=number, default=1}
 GravityOverride {type=Vector}
 IsCelebrationEmote {type=boolean}
 IschangedWeapon {type=boolean}
 isCustomSkin {type=boolean}
 isGunOut {type=boolean}
 IsStickOfTruth {type=boolean}
 IsWearingGauntlets {type=boolean}
 Output {type=Object, default=null}
 SpringConstant {type=number}
]
