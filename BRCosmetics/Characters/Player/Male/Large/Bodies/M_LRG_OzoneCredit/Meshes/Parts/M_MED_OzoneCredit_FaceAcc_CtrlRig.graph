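# Control Rig for the OzoneCredit face accessory, driving a chest-mounted joystick and two small arms that follow it.
# Constraints pin dyn spine, neck and head bones to the body and blend the joystick base across spine bones.
# Smoothed Pub_Move inputs (Forward feeds Scale, not Value) and DeltaRotation swing/twist weights from head and spine_05 tilt the joystick.
# Arm controls constrained to joystick and spine then drive stretchy two-bone IK.
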ControlRigBlueprint /BRCosmetics/Characters/Player/Male/Large/Bodies/M_LRG_OzoneCredit/Meshes/Parts/M_MED_OzoneCredit_FaceAcc_CtrlRig
graphs [
 CollapseNode_ContainedGraph {nodes [
   n1 = RigVMFunctionEntryNode()
   &RigUnit_ParentConstraint(
    AdvancedSettings=(RotationOrderForFilter=XZY)
    Child=(Type=Bone,Name="base_joystick")
    Parents=((Item=(Type=Bone,Name="spine_04"),Weight=0.0515),(Item=(Type=Bone,Name="spine_03"),Weight=0.2394),(Item=(Type=Bone,Name="spine_05"),Weight=0.7091))
    Weight=1
    ExecutePin=@n1.ExecutePin
   )
   RigUnit_GetTransform = &RigUnit_GetTransform(Item=(Type=Control,Name="dyn_arm_03_l_ctrl"), bInitial=False)
   > /Script/ControlRig.RigUnit_TwoBoneIKSimplePerItem(
    DebugSettings=(bEnabled=False,Scale=10,WorldOffset=(Rotation=(X=0,Y=0,Z=0,W=1),Translation=(X=0,Y=0,Z=0),Scale3D=(X=1,Y=1,Z=1)))
    EffectorItem=(Type=Bone,Name="dyn_arm_03_l")
    ItemA=(Type=Bone,Name="dyn_arm_01_l")
    ItemALength=0
    ItemB=(Type=Bone,Name="dyn_arm_02_l")
    ItemBLength=0
    PoleVectorKind=Direction
    PoleVectorSpace=(Type=Bone,Name="None")
    PrimaryAxis=(X=1,Y=0,Z=0)
    SecondaryAxis=(X=0,Y=0,Z=1)
    SecondaryAxisWeight=1
    StretchMaximumRatio=1.5
    StretchStartRatio=0.75
    Weight=1
    bEnableStretch=true
    bPropagateToChildren=true
    Effector=@RigUnit_GetTransform.Transform
    PoleVector=&RigUnit_GetTransform(Item=(Type=Control,Name="dyn_arm_02_l_ctrl"), bInitial=False).Transform.Translation
   ) {is=struct RigVMUnitNode}
   > n2 = &RigUnit_TwoBoneIKSimplePerItem(
    DebugSettings=(bEnabled=False,Scale=10,WorldOffset=(Rotation=(X=0,Y=0,Z=0,W=1),Translation=(X=0,Y=0,Z=0),Scale3D=(X=1,Y=1,Z=1)))
    EffectorItem=(Type=Bone,Name="dyn_arm_03_r")
    ItemA=(Type=Bone,Name="dyn_arm_01_r")
    ItemALength=0
    ItemB=(Type=Bone,Name="dyn_arm_02_r")
    ItemBLength=0
    PoleVectorKind=Direction
    PoleVectorSpace=(Type=Bone,Name="None")
    PrimaryAxis=(X=-1,Y=0,Z=0)
    SecondaryAxis=(X=0,Y=0,Z=-1)
    SecondaryAxisWeight=1
    StretchMaximumRatio=1.5
    StretchStartRatio=0.75
    Weight=1
    bEnableStretch=true
    bPropagateToChildren=true
    Effector=&RigUnit_GetTransform(Item=(Type=Control,Name="dyn_arm_03_r_ctrl"), bInitial=False).Transform
    PoleVector=&RigUnit_GetTransform(Item=(Type=Control,Name="dyn_arm_02_r_ctrl"), bInitial=False).Transform.Translation
   )
   RigVMFunctionReturnNode(ExecutePin=@n2.ExecutePin)
  ]}
 New Function_ContainedGraph {nodes [
   n1 = RigVMFunctionEntryNode()
   RigVMFunction_MathBoolToInteger = &RigVMFunction_MathBoolToInteger(Value=@n1.onlyTwist)
   RigVMFunction_MathQuaternionSwingTwist = &RigVMFunction_MathQuaternionSwingTwist(
    TwistAxis=@n1.twistAxis
    Input=&RigUnit_GetRelativeTransformForItem(bChildInitial=False, bParentInitial=False, Child=@n1.child, Parent=@n1.parent).RelativeTransform.Rotation
   )
   n2 = RigVMDispatchNode(
    Values=((),())
    Index=@RigVMFunction_MathBoolToInteger.Result
    Values.0=@RigVMFunction_MathQuaternionSwingTwist.Swing
    Values.1=@RigVMFunction_MathQuaternionSwingTwist.Twist
   ) {template="DISPATCH_RigVMDispatch_SelectInt32(in Index,in Values,out Result)"}
   RigVMFunction_MathQuaternionToEuler = &RigVMFunction_MathQuaternionToEuler(Value=@n2.Result, RotationOrder=@n1.Rotation Order)
   > DISPATCH_RigVMDispatch_Print = RigVMDispatchNode(ScreenColor=(R=1,G=1,B=1,A=1), ScreenDuration=1.5, Enabled=@n1.debugRotation, Value=@RigVMFunction_MathQuaternionToEuler.Result) {template="DISPATCH_RigVMDispatch_Print(in Prefix,in Value,in Enabled,in ScreenDuration,in ScreenColor)"}
   Cast = RigVMDispatchNode(Value=@n1.triggerAxis) {template="DISPATCH_RigVMDispatch_CastEnumToInt(in Value,out Result)"}
   DISPATCH_RigVMDispatch_SelectInt32 = RigVMDispatchNode(
    Values=(0,0,0)
    Index=@Cast.Result
    Values.0=@RigVMFunction_MathQuaternionToEuler.Result.X
    Values.1=@RigVMFunction_MathQuaternionToEuler.Result.Y
    Values.2=@RigVMFunction_MathQuaternionToEuler.Result.Z
   ) {template="DISPATCH_RigVMDispatch_SelectInt32(in Index,in Values,out Result)"}
   Remap = &RigVMFunction_MathDoubleRemap(
    bClamp=False
    Value=@DISPATCH_RigVMDispatch_SelectInt32.Result
    TargetMinimum=@n1.outputMinMax.X
    TargetMaximum=@n1.outputMinMax.Y
    SourceMinimum=@n1.rotationMinMax.X
    SourceMaximum=@n1.rotationMinMax.Y
   )
   Absolute = &RigVMFunction_MathDoubleAbs(Value=@Remap.Result)
   Multiply = &RigVMFunction_MathDoubleMul(
    A=&RigVMFunction_MathDoubleClamp(Value=RigVMDispatchNode(True=@Absolute.Result, Condition=@n1.absoluteValue, False=@Remap.Result) {template="DISPATCH_RigVMDispatch_If(in Condition,in True,in False,out Result)"}.Result).Result
    B=@n1.premult
   )
   > RigVMFunctionReturnNode(outputQuat=@n2.Result, outputWeight=@Multiply.Result)
  ]}
 RigVMFunctionLibrary {nodes [DeltaRotation v4 = RigVMCollapseNode(
    ExecuteContext=()
    Rotation Order=ZYX
    child=(Type=Bone,Name="None")
    outputMinMax=(X=0,Y=1)
    parent=(Type=Bone,Name="None")
    premult=1
    rotationMinMax=(X=0,Y=0)
    twistAxis=(X=1,Y=0,Z=0)
   )]}
 RigVMModel {nodes [
   RigUnit_BeginExecution = &RigUnit_BeginExecution()
   &RigUnit_ParentConstraint(
    AdvancedSettings=(InterpolationType=Average,RotationOrderForFilter=XZY)
    Child=(Type=Bone,Name="dyn_spine_04")
    Parents=((Item=(Type=Bone,Name="spine_04"),Weight=1))
    Weight=1
    ExecutePin=@RigUnit_BeginExecution.ExecutePin
   )
   > &RigUnit_ParentConstraint(
    AdvancedSettings=(InterpolationType=Average,RotationOrderForFilter=XZY)
    Child=(Type=Bone,Name="dyn_spine_05")
    Parents=((Item=(Type=Bone,Name="spine_05"),Weight=1))
    Weight=1
   )
   > &RigUnit_ParentConstraint(
    AdvancedSettings=(InterpolationType=Average,RotationOrderForFilter=XZY)
    Child=(Type=Bone,Name="dyn_neck_01")
    Parents=((Item=(Type=Bone,Name="neck_01"),Weight=1))
    Weight=1
   )
   > &RigUnit_ParentConstraint(
    AdvancedSettings=(InterpolationType=Average,RotationOrderForFilter=XZY)
    Child=(Type=Bone,Name="dyn_neck_02")
    Parents=((Item=(Type=Bone,Name="neck_02"),Weight=1))
    Weight=1
   )
   > &RigUnit_ParentConstraint(
    AdvancedSettings=(InterpolationType=Average,RotationOrderForFilter=XZY)
    Child=(Type=Bone,Name="dyn_head")
    Parents=((Item=(Type=Bone,Name="head"),Weight=1))
    Weight=1
   )
   > &RigUnit_ParentConstraint(
    AdvancedSettings=(RotationOrderForFilter=XZY)
    Child=(Type=Bone,Name="base_joystick")
    Parents=((Item=(Type=Bone,Name="spine_04"),Weight=0.0515),(Item=(Type=Bone,Name="spine_03"),Weight=0.2394),(Item=(Type=Bone,Name="spine_05"),Weight=0.7091))
    Weight=1
   )
   AlphaInterp = &RigVMFunction_AlphaInterp(InterpSpeedDecreasing=5, InterpSpeedIncreasing=5, Value=0, bClampResult=False, Scale=RigVMVariableNode() {variable=Pub_MoveForward}.Value)
   > &RigUnit_ModifyTransforms(
    ItemToModify="((Item=(Type=Bone,Name=\"dyn_joystick\"),Transform=(Rotation=(X=0,Y=0,Z=0.1736,W=0.9848),Translation=(X=0,Y=0,Z=0),Scale3D=(X=1,Y=1,Z=1))),(Item=(Type=Bone,Name=\"dyn_neck_01\"),Transform=(Rotation=(X=0,Y=0,Z=0.1736,W=0.9848),Translation=(X=0,Y=0,Z=0),Scale3D= ...(truncated)"
    WeightMinimum=0
    Weight=@AlphaInterp.Result
   )
   n1 = &RigVMFunction_AlphaInterp(InterpSpeedDecreasing=5, InterpSpeedIncreasing=5, Scale=1, bClampResult=False, Value=RigVMVariableNode() {variable=Pub_MoveBackward}.Value)
   > &RigUnit_ModifyTransforms(
    ItemToModify="((Item=(Type=Bone,Name=\"dyn_joystick\"),Transform=(Rotation=(X=0,Y=0,Z=-0.0872,W=0.9962),Translation=(X=0,Y=0,Z=0),Scale3D=(X=1,Y=1,Z=1))),(Item=(Type=Bone,Name=\"dyn_neck_01\"),Transform=(Rotation=(X=0,Y=0,Z=-0.0436,W=0.999),Translation=(X=0,Y=0,Z=0),Scale3D ...(truncated)"
    WeightMinimum=0
    Weight=@n1.Result
   )
   n2 = &RigVMFunction_AlphaInterp(InterpSpeedDecreasing=5, InterpSpeedIncreasing=5, Scale=1, bClampResult=False, Value=RigVMVariableNode() {variable=Pub_MoveLeft}.Value)
   > &RigUnit_ModifyTransforms(
    ItemToModify=((Item=(Type=Bone,Name="dyn_joystick"),Transform=(Rotation=(X=0,Y=-0.1305,Z=0,W=0.9914),Translation=(X=0,Y=0,Z=0),Scale3D=(X=1,Y=1,Z=1))))
    WeightMinimum=0
    Weight=@n2.Result
   )
   n3 = &RigVMFunction_AlphaInterp(InterpSpeedDecreasing=5, InterpSpeedIncreasing=5, Scale=1, bClampResult=False, Value=RigVMVariableNode() {variable=Pub_MoveRight}.Value)
   > &RigUnit_ModifyTransforms(
    ItemToModify=((Item=(Type=Bone,Name="dyn_joystick"),Transform=(Rotation=(X=0,Y=0.1305,Z=0,W=0.9914),Translation=(X=0,Y=0,Z=0),Scale3D=(X=1,Y=1,Z=1))))
    WeightMinimum=0
    Weight=@n3.Result
   )
   > DeltaRotation v4 = RigVMFunctionReferenceNode(
    Rotation Order=ZYX
    child=(Type=Bone,Name="head")
    debugRotation=false
    outputMinMax=(X=0,Y=1)
    parent=(Type=Null,Name="head_null")
    premult=1
    rotationMinMax=(X=0,Y=-20)
    triggerAxis=Z
    twistAxis=(X=1,Y=0,Z=0)
   )
   > &RigUnit_ModifyTransforms(
    ItemToModify=((Item=(Type=Bone,Name="dyn_joystick"),Transform=(Rotation=(X=0,Y=0,Z=-0.1305,W=0.9914),Translation=(X=0,Y=0,Z=0),Scale3D=(X=1,Y=1,Z=1))))
    WeightMinimum=0
    Weight=@"DeltaRotation v4".outputWeight
   )
   > DeltaRotation v4_1 = RigVMFunctionReferenceNode(
    Rotation Order=ZYX
    child=(Type=Bone,Name="head")
    debugRotation=false
    outputMinMax=(X=0,Y=1)
    parent=(Type=Null,Name="head_null")
    premult=1
    rotationMinMax=(X=0,Y=30)
    triggerAxis=Z
    twistAxis=(X=1,Y=0,Z=0)
   )
   > &RigUnit_ModifyTransforms(
    ItemToModify=((Item=(Type=Bone,Name="dyn_joystick"),Transform=(Rotation=(X=0,Y=0,Z=0.2164,W=0.9763),Translation=(X=0,Y=0,Z=0),Scale3D=(X=1,Y=1,Z=1))))
    WeightMinimum=0
    Weight=@"DeltaRotation v4_1".outputWeight
   )
   > DeltaRotation v4_2 = RigVMFunctionReferenceNode(
    Rotation Order=ZYX
    child=(Type=Bone,Name="spine_05")
    debugRotation=false
    onlyTwist=true
    outputMinMax=(X=0,Y=1)
    parent=(Type=Null,Name="spine_05_null")
    premult=1
    rotationMinMax=(X=0,Y=-10)
    triggerAxis=X
    twistAxis=(X=1,Y=0,Z=0)
   )
   > &RigUnit_ModifyTransforms(
    ItemToModify=((Item=(Type=Bone,Name="dyn_joystick"),Transform=(Rotation=(X=-0.0114,Y=0.13,Z=0.0864,W=0.9877),Translation=(X=0,Y=0,Z=0),Scale3D=(X=1,Y=1,Z=1))))
    WeightMinimum=0
    Weight=@"DeltaRotation v4_2".outputWeight
   )
   > DeltaRotation v4_3 = RigVMFunctionReferenceNode(
    Rotation Order=ZYX
    child=(Type=Bone,Name="spine_05")
    debugRotation=false
    onlyTwist=true
    outputMinMax=(X=0,Y=1)
    parent=(Type=Null,Name="spine_05_null")
    premult=1
    rotationMinMax=(X=0,Y=10)
    triggerAxis=X
    twistAxis=(X=1,Y=0,Z=0)
   )
   > &RigUnit_ModifyTransforms(
    ItemToModify=((Item=(Type=Bone,Name="dyn_joystick"),Transform=(Rotation=(X=0.0076,Y=-0.0868,Z=0.0868,W=0.9924),Translation=(X=0,Y=0,Z=0),Scale3D=(X=1,Y=1,Z=1))))
    WeightMinimum=0
    Weight=@"DeltaRotation v4_3".outputWeight
   )
   > &RigUnit_ParentConstraint(
    AdvancedSettings=(InterpolationType=Average,RotationOrderForFilter=XZY)
    Child=(Type=Control,Name="dyn_arm_03_l_ctrl")
    Parents=((Item=(Type=Bone,Name="dyn_joystick"),Weight=1))
    Weight=1
   )
   > &RigUnit_ParentConstraint(
    AdvancedSettings=(InterpolationType=Average,RotationOrderForFilter=XZY)
    Child=(Type=Control,Name="dyn_arm_03_r_ctrl")
    Parents=((Item=(Type=Bone,Name="dyn_joystick"),Weight=1))
    Weight=1
    bMaintainOffset=true
   )
   > &RigUnit_ModifyTransforms(
    ItemToModify="((Item=(Type=Control,Name=\"dyn_arm_03_l_ctrl\"),Transform=(Rotation=(X=0,Y=0.0872,Z=0,W=0.9962),Translation=(X=0,Y=0,Z=0),Scale3D=(X=1,Y=1,Z=1))),(Item=(Type=Control,Name=\"dyn_arm_03_r_ctrl\"),Transform=(Rotation=(X=-0.0151,Y=0.0858,Z=0.173,W=0.9811),Transla ...(truncated)"
    Weight=1
    WeightMinimum=0
   )
   > &RigUnit_ParentConstraint(
    AdvancedSettings=(InterpolationType=Average,RotationOrderForFilter=XZY)
    Child=(Type=Control,Name="dyn_arm_02_l_ctrl")
    Parents=((Item=(Type=Bone,Name="dyn_joystick"),Weight=1))
    Weight=1
    bMaintainOffset=true
   )
   > n4 = &RigUnit_ParentConstraint(
    AdvancedSettings=(InterpolationType=Average,RotationOrderForFilter=XZY)
    Child=(Type=Control,Name="dyn_arm_02_r_ctrl")
    Parents=((Item=(Type=Bone,Name="spine_05"),Weight=1))
    Weight=1
    bMaintainOffset=true
   )
   Ik_Arms_Setup = RigVMCollapseNode(ExecutePin=@n4.ExecutePin)
  ]}
]
hierarchy [
 root {loc=<0 0 0>, rot=<0 0 0 1>}
 pelvis {parent=root, loc=<0 -0.7321 92.4218>, rot=<0.0127 0.707 -0.0127 -0.707>}
 spine_01 {parent=pelvis, loc=<5.6002 -0.5128 0>, rot=<0 0 0.034 -0.9994>}
 spine_02 {parent=spine_01, loc=<8.0756 0.0139 0>, rot=<0 0 0.0238 -0.9997>}
 spine_03 {parent=spine_02, loc=<7.0395 0 0>, rot=<0 0 0.0149 -0.9999>}
 spine_04 {parent=spine_03, loc=<8.1023 0 0>, rot=<0 0 -0.0132 -0.9999>}
 spine_05 {parent=spine_04, loc=<10.0141 0 0>, rot=<0 0 -0.0015 -1>}
 neck_01 {parent=spine_05, loc=<10.815 1.9196 0>, rot=<0 0 -0.1864 -0.9825>}
 neck_02 {parent=neck_01, loc=<4.7347 0.3045 0>, rot=<0 0 0.0363 -0.9993>}
 head {parent=neck_02, loc=<5.7632 0.5552 0>, rot=<0 0 0.115 -0.9934>}
 head_null {parent=spine_05, type=Null, loc=<20.4455 6.1834 0>, rot=<0 0 -0.036 -0.9994>}
 dyn_spine_04 {parent=spine_04, loc=<0 0 0>, rot=<0 0 0 1>}
 dyn_spine_05 {parent=dyn_spine_04, loc=<10.0141 0 0>, rot=<0 0 -0.0015 -1>}
 dyn_neck_01 {parent=dyn_spine_05, loc=<10.815 1.9196 0>, rot=<0 0 -0.1864 -0.9825>}
 dyn_neck_02 {parent=dyn_neck_01, loc=<4.7347 0.3045 0>, rot=<0 0 0.0363 -0.9993>}
 dyn_head {parent=dyn_neck_02, loc=<5.7632 0.5552 0>, rot=<0 0 0.115 -0.9934>}
 dyn_hair_A {parent=dyn_head, loc=<21.385 13.5227 -1.6069>, rot=<-0.1708 0.3059 0.8216 0.4498>}
 dyn_hair_B {parent=dyn_head, loc=<21.3517 13.4182 1.7813>, rot=<0.2289 -0.3672 0.769 0.4706>}
 dyn_hair_C {parent=dyn_head, loc=<22.4949 9.7903 3.3949>, rot=<0.4086 -0.4147 0.7086 0.3986>}
 dyn_hair_D {parent=dyn_head, loc=<23.7992 6.4072 1.9961>, rot=<0.807 -0.5126 0.2072 0.2075>}
 dyn_hair_E {parent=dyn_head, loc=<23.7582 6.7696 -2.2698>, rot=<0.735 -0.5635 -0.2319 -0.2974>}
 dyn_hair_F {parent=dyn_head, loc=<22.3125 9.9743 -3.6857>, rot=<-0.4687 0.3543 0.7213 0.3667>}
 dyn_helmetDeformer {parent=dyn_head, loc=<14.274 2.508 0>, rot=<0 0 -0.2688 -0.9632>}
 base_arm_01_l {parent=dyn_neck_01, loc=<5.9742 3.4863 -8.8873>, rot=<-0.5907 -0.7949 0.0295 0.1351>}
 dyn_arm_01_l {parent=base_arm_01_l, loc=<0 0 0>, rot=<0 0 0 1>}
 dyn_arm_02_l {parent=dyn_arm_01_l, loc=<7.3508 0 0>, rot=<0 -0.1493 0 -0.9888>}
 dyn_arm_03_l {parent=dyn_arm_02_l, loc=<5.1294 0 0>, rot=<0 0 0 1>}
 base_arm_01_r {parent=dyn_neck_01, loc=<5.9746 3.4862 8.8873>, rot=<0.7949 -0.5907 0.1351 -0.0295>}
 dyn_arm_01_r {parent=base_arm_01_r, loc=<0 0 0>, rot=<0 0 0 1>}
 dyn_arm_02_r {parent=dyn_arm_01_r, loc=<-7.3509 0 0>, rot=<0 -0.1493 0 -0.9888>}
 dyn_arm_03_r {parent=dyn_arm_02_r, loc=<-5.1295 0.0001 -0.0001>, rot=<0 0 0 1>}
 dyn_R_shoeBase {parent=dyn_spine_05, loc=<3.5438 16.0059 6.8142>, rot=<0.7065 0.0284 -0.7065 0.0284>}
 dyn_R_foot_shoeLace_A {parent=dyn_R_shoeBase, loc=<-0.862 0.147 -0.063>, rot=<-0.3167 0.183 0.1033 -0.925>}
 dyn_R_foot_shoeLace_B {parent=dyn_R_shoeBase, loc=<0.8045 0.1695 -0.354>, rot=<0.0738 -0.8926 -0.3007 0.3276>}
 dyn_R_foot_laceTies {parent=dyn_R_shoeBase, loc=<0.0468 -0.3903 -0.234>, rot=<0.3476 0.7024 -0.2657 -0.5614>}
 dyn_L_shoeBase {parent=dyn_spine_05, loc=<3.5441 16.006 -6.8142>, rot=<0.0284 -0.7065 -0.0284 -0.7065>}
 dyn_L_foot_shoeLace_A {parent=dyn_L_shoeBase, loc=<0.862 -0.147 0.0625>, rot=<-0.3167 0.183 0.1033 -0.925>}
 dyn_L_foot_shoeLace_B {parent=dyn_L_shoeBase, loc=<-0.8045 -0.1695 0.3538>, rot=<0.0738 -0.8926 -0.3007 0.3276>}
 dyn_L_foot_laceTies {parent=dyn_L_shoeBase, loc=<-0.0468 0.3903 0.2334>, rot=<0.3476 0.7024 -0.2657 -0.5614>}
 base_joystick {parent=dyn_spine_04, loc=<8.4654 13.7971 0>, rot=<0 0 -0.2498 -0.9683>}
 dyn_joystick {parent=base_joystick, loc=<0 0 0>, rot=<0 0 0 1>}
 dyn_joystickHead {parent=dyn_joystick, loc=<10.0186 0.1885 0>, rot=<-0.148 -0.6914 0.148 -0.6914>}
 spine_05_null {parent=spine_04, type=Null, loc=<10.0141 0 0>, rot=<0 0 -0.0015 -1>}
 dyn_arm_03_l_ctrl {controlType=EULER_TRANSFORM, type=Control, loc=<-114.7547 34.6112 57.5344>, rot=<0.7919 -0.5182 0.0621 0.317>}
 dyn_arm_02_r_ctrl {controlType=EULER_TRANSFORM, type=Control, loc=<0 0 0>, rot=<0 0 0 1>}
 dyn_arm_03_r_ctrl {controlType=EULER_TRANSFORM, type=Control, loc=<-0.0966 12.6533 140.7231>, rot=<-0.4142 0.3472 -0.7794 0.317>}
 dyn_arm_02_l_ctrl {controlType=EULER_TRANSFORM, type=Control, loc=<0 0 0>, rot=<0 0 0 1>}
]
variables [
 Pub_MoveBackward {type=double, default=0}
 Pub_MoveForward {type=double, default=0}
 Pub_MoveLeft {type=double, default=0}
 Pub_MoveRight {type=double, default=0}
]
