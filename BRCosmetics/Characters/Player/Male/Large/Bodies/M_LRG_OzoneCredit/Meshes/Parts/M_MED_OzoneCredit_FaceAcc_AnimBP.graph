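# Post-process style AnimBlueprint for the OzoneCredit character's FaceAcc cosmetic part, adding secondary motion to a pose copied from the attached parent mesh.
# A Control Rig takes four directional movement weights that the EventGraph derives from PlayerMovement each update.
# The pose then passes through a dyn_head-to-head constraint, spring-based AnimDynamics for hair and shoelace bones, and fixed neck-bone offsets.
# JoystickRotator, movementAlpha and IsHiddenHead are set but unused by the AnimGraph.
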AnimBlueprint /BRCosmetics/Characters/Player/Male/Large/Bodies/M_LRG_OzoneCredit/Meshes/Parts/M_MED_OzoneCredit_FaceAcc_AnimBP
target_skeleton=&/M_MED_OzoneCredit_FaceAcc_Skeleton
graphs [
 AnimGraph {nodes [
   n1 = Control Rig!(
    Source=Copy Pose From Mesh() {bUseAttachedParent=true}
    Alpha
    Pub_MoveForward:double=Get moveForward() -> (moveForward:double)
    Pub_MoveBackward:double=Get moveBackward() -> (moveBackward:double)
    Pub_MoveLeft:double=Get strafeLeft float() -> (strafeLeft float:double)
    Pub_MoveRight:double=Get strafeRight float() -> (strafeRight float:double)
   ) -> (Pose) {controlRigAssetReference {blueprintRigClass=&/M_MED_OzoneCredit_FaceAcc_CtrlRig.M_MED_OzoneCredit_FaceAcc_CtrlRig_C}}
   n2 = "Constraint - dyn_head (P)"(ComponentPose=Local To Component(LocalPose=@n1), Alpha=1, ConstraintWeights_0=1) {boneToModify {boneName=dyn_head}, constraintSetup [{targetBone {boneName=head}}], constraintWeights [1]}
   n3 = Anim Dynamics - Bone: dyn_hair_A(ComponentPose=@n2, Alpha=1, ExternalForce="0,0,0") {
    angularDampingOverride=1
    angularSpringConstant=20000
    bAngularSpring=true
    bOverrideAngularDamping=true
    bUseGravityOverride=true
    gravityScale=0.35
    simulationSpace=RootRelative
    boundBone {boneName=dyn_hair_A}
    componentAppliedLinearAccClamp=<4000 4000 4000>
    componentLinearAccScale=<1 1 1>
    componentLinearVelScale=<1 1 1>
    gravityOverride {z=-800}
    physicsBodyDefinitions [{
      boundBone {boneName=dyn_hair_A}
      boxExtents=<21 21 21>
      constraintSetup {angularLimitsMax=<0 4 0>, angularLimitsMin=<0 -4 -60>, angularTarget=<1 0 0>}
      localJointOffset=<4 0 0>
     }]
    simSpaceSettings {maxAngularAcceleration=35, maxAngularVelocity=7, simSpaceAngularAlpha=1}
   }
   n4 = Anim Dynamics - Bone: dyn_hair_B(ComponentPose=@n3, Alpha=1, ExternalForce="0,0,0") {
    angularDampingOverride=1
    angularSpringConstant=20000
    bAngularSpring=true
    bOverrideAngularDamping=true
    bUseGravityOverride=true
    gravityScale=0.35
    simulationSpace=RootRelative
    boundBone {boneName=dyn_hair_B}
    componentAppliedLinearAccClamp=<4000 4000 4000>
    componentLinearAccScale=<1 1 1>
    componentLinearVelScale=<1 1 1>
    gravityOverride {z=-800}
    physicsBodyDefinitions [{
      boundBone {boneName=dyn_hair_B}
      boxExtents=<25 25 25>
      constraintSetup {angularLimitsMax=<0 4 0>, angularLimitsMin=<0 -4 -60>, angularTarget=<1 0 0>}
      localJointOffset=<4 0 0>
     }]
    simSpaceSettings {maxAngularAcceleration=35, maxAngularVelocity=7, simSpaceAngularAlpha=1}
   }
   n5 = Anim Dynamics - Bone: dyn_hair_C(ComponentPose=@n4, Alpha=1, ExternalForce="0,0,0") {
    angularDampingOverride=1
    angularSpringConstant=20000
    bAngularSpring=true
    bOverrideAngularDamping=true
    bUseGravityOverride=true
    gravityScale=0.35
    simulationSpace=RootRelative
    boundBone {boneName=dyn_hair_C}
    componentAppliedLinearAccClamp=<4000 4000 4000>
    componentLinearAccScale=<1 1 1>
    componentLinearVelScale=<1 1 1>
    gravityOverride {z=-800}
    physicsBodyDefinitions [{
      boundBone {boneName=dyn_hair_C}
      boxExtents=<25 25 25>
      constraintSetup {angularLimitsMax=<0 4 0>, angularLimitsMin=<0 -4 -60>, angularTarget=<1 0 0>}
      localJointOffset=<4 0 0>
     }]
    simSpaceSettings {maxAngularAcceleration=35, maxAngularVelocity=7, simSpaceAngularAlpha=1}
   }
   n6 = Anim Dynamics - Bone: dyn_hair_D(ComponentPose=@n5, Alpha=1, ExternalForce="0,0,0") {
    angularDampingOverride=1
    angularSpringConstant=20000
    bAngularSpring=true
    bOverrideAngularDamping=true
    bUseGravityOverride=true
    gravityScale=0.35
    simulationSpace=RootRelative
    boundBone {boneName=dyn_hair_D}
    componentAppliedLinearAccClamp=<4000 4000 4000>
    componentLinearAccScale=<1 1 1>
    componentLinearVelScale=<1 1 1>
    gravityOverride {z=-800}
    physicsBodyDefinitions [{
      boundBone {boneName=dyn_hair_D}
      boxExtents=<27 27 27>
      constraintSetup {angularLimitsMax=<0 4 0>, angularLimitsMin=<0 -4 -60>, angularTarget=<1 0 0>}
      localJointOffset=<4 0 0>
     }]
    simSpaceSettings {maxAngularAcceleration=35, maxAngularVelocity=7, simSpaceAngularAlpha=1}
   }
   n7 = Anim Dynamics - Bone: dyn_hair_E(ComponentPose=@n6, Alpha=1, ExternalForce="0,0,0") {
    angularDampingOverride=1
    angularSpringConstant=20000
    bAngularSpring=true
    bOverrideAngularDamping=true
    bUseGravityOverride=true
    gravityScale=0.35
    simulationSpace=RootRelative
    boundBone {boneName=dyn_hair_E}
    componentAppliedLinearAccClamp=<4000 4000 4000>
    componentLinearAccScale=<1 1 1>
    componentLinearVelScale=<1 1 1>
    gravityOverride {z=-800}
    physicsBodyDefinitions [{
      boundBone {boneName=dyn_hair_E}
      boxExtents=<21 21 21>
      constraintSetup {angularLimitsMax=<0 4 0>, angularLimitsMin=<0 -4 -60>, angularTarget=<1 0 0>}
      localJointOffset=<4 0 0>
     }]
    simSpaceSettings {maxAngularAcceleration=35, maxAngularVelocity=7, simSpaceAngularAlpha=1}
   }
   n8 = Anim Dynamics - Bone: dyn_hair_F(ComponentPose=@n7, Alpha=1, ExternalForce="0,0,0") {
    angularDampingOverride=1
    angularSpringConstant=20000
    bAngularSpring=true
    bOverrideAngularDamping=true
    bUseGravityOverride=true
    gravityScale=0.35
    simulationSpace=RootRelative
    boundBone {boneName=dyn_hair_F}
    componentAppliedLinearAccClamp=<4000 4000 4000>
    componentLinearAccScale=<1 1 1>
    componentLinearVelScale=<1 1 1>
    gravityOverride {z=-800}
    physicsBodyDefinitions [{
      boundBone {boneName=dyn_hair_F}
      boxExtents=<25 25 25>
      constraintSetup {angularLimitsMax=<0 4 0>, angularLimitsMin=<0 -4 -60>, angularTarget=<1 0 0>}
      localJointOffset=<4 0 0>
     }]
    simSpaceSettings {maxAngularAcceleration=35, maxAngularVelocity=7, simSpaceAngularAlpha=1}
   }
   n9 = Anim Dynamics - Bone: dyn_L_foot_shoeLace_A(ComponentPose=@n8, Alpha=1, ExternalForce="0,0,0") {
    angularDampingOverride=1
    angularSpringConstant=6000
    bAngularSpring=true
    bOverrideAngularDamping=true
    bUseGravityOverride=true
    gravityScale=0.35
    simulationSpace=RootRelative
    boundBone {boneName=dyn_L_foot_shoeLace_A}
    componentAppliedLinearAccClamp=<4000 4000 4000>
    componentLinearAccScale=<1 1 1>
    componentLinearVelScale=<1 1 1>
    gravityOverride {z=-800}
    physicsBodyDefinitions [{
      boundBone {boneName=dyn_L_foot_shoeLace_A}
      boxExtents=<11 11 11>
      constraintSetup {angularLimitsMax=<0 0 30>, angularLimitsMin=<0 0 -30>, angularTarget=<1 0 0>}
      localJointOffset=<2 0 0>
     }]
    simSpaceSettings {maxAngularAcceleration=35, maxAngularVelocity=7, simSpaceAngularAlpha=1}
   }
   n10 = Anim Dynamics - Bone: dyn_R_foot_shoeLace_A(ComponentPose=@n9, Alpha=1, ExternalForce="0,0,0") {
    angularDampingOverride=1
    angularSpringConstant=6000
    bAngularSpring=true
    bOverrideAngularDamping=true
    bUseGravityOverride=true
    gravityScale=0.35
    simulationSpace=RootRelative
    boundBone {boneName=dyn_R_foot_shoeLace_A}
    componentAppliedLinearAccClamp=<4000 4000 4000>
    componentLinearAccScale=<1 1 1>
    componentLinearVelScale=<1 1 1>
    gravityOverride {z=-800}
    physicsBodyDefinitions [{
      boundBone {boneName=dyn_R_foot_shoeLace_A}
      boxExtents=<11 11 11>
      constraintSetup {angularLimitsMax=<0 0 30>, angularLimitsMin=<0 0 -30>, angularTarget=<1 0 0>}
      localJointOffset=<-2 0 0>
     }]
    simSpaceSettings {maxAngularAcceleration=35, maxAngularVelocity=7, simSpaceAngularAlpha=1}
   }
   n11 = Anim Dynamics - Bone: dyn_L_foot_shoeLace_B(ComponentPose=@n10, Alpha=1, ExternalForce="0,0,0") {
    angularDampingOverride=1
    angularSpringConstant=6000
    bAngularSpring=true
    bOverrideAngularDamping=true
    bUseGravityOverride=true
    gravityScale=0.35
    simulationSpace=RootRelative
    boundBone {boneName=dyn_L_foot_shoeLace_B}
    componentAppliedLinearAccClamp=<4000 4000 4000>
    componentLinearAccScale=<1 1 1>
    componentLinearVelScale=<1 1 1>
    gravityOverride {z=-800}
    physicsBodyDefinitions [{
      boundBone {boneName=dyn_L_foot_shoeLace_B}
      boxExtents=<11 11 11>
      constraintSetup {angularLimitsMax=<0 0 30>, angularLimitsMin=<0 0 -30>, angularTarget=<1 0 0>}
      localJointOffset=<2 0 0>
     }]
    simSpaceSettings {maxAngularAcceleration=35, maxAngularVelocity=7, simSpaceAngularAlpha=1}
   }
   n12 = Anim Dynamics - Bone: dyn_R_foot_shoeLace_B(ComponentPose=@n11, Alpha=1, ExternalForce="0,0,0") {
    angularDampingOverride=1
    angularSpringConstant=6000
    bAngularSpring=true
    bOverrideAngularDamping=true
    bUseGravityOverride=true
    gravityScale=0.35
    simulationSpace=RootRelative
    boundBone {boneName=dyn_R_foot_shoeLace_B}
    componentAppliedLinearAccClamp=<4000 4000 4000>
    componentLinearAccScale=<1 1 1>
    componentLinearVelScale=<1 1 1>
    gravityOverride {z=-800}
    physicsBodyDefinitions [{
      boundBone {boneName=dyn_R_foot_shoeLace_B}
      boxExtents=<11 11 11>
      constraintSetup {angularLimitsMax=<0 0 30>, angularLimitsMin=<0 0 -30>, angularTarget=<1 0 0>}
      localJointOffset=<-2 0 0>
     }]
    simSpaceSettings {maxAngularAcceleration=35, maxAngularVelocity=7, simSpaceAngularAlpha=1}
   }
   n13 = Anim Dynamics - Bone: dyn_L_foot_laceTies(ComponentPose=@n12, Alpha=1, ExternalForce="0,0,0") {
    angularDampingOverride=1
    angularSpringConstant=6000
    bAngularSpring=true
    bOverrideAngularDamping=true
    bUseGravityOverride=true
    gravityScale=0.35
    simulationSpace=RootRelative
    boundBone {boneName=dyn_L_foot_laceTies}
    componentAppliedLinearAccClamp=<4000 4000 4000>
    componentLinearAccScale=<1 1 1>
    componentLinearVelScale=<1 1 1>
    gravityOverride {z=-800}
    physicsBodyDefinitions [{
      boundBone {boneName=dyn_L_foot_laceTies}
      boxExtents=<11 11 11>
      constraintSetup {angularLimitsMax=<0 0 0>, angularLimitsMin=<0 0 -30>, angularTarget=<1 0 0>}
      localJointOffset=<2 0 0>
     }]
    simSpaceSettings {maxAngularAcceleration=35, maxAngularVelocity=7, simSpaceAngularAlpha=1}
   }
   n14 = Anim Dynamics - Bone: dyn_R_foot_laceTies(ComponentPose=@n13, Alpha=1, ExternalForce="0,0,0") {
    angularDampingOverride=1
    angularSpringConstant=6000
    bAngularSpring=true
    bOverrideAngularDamping=true
    bUseGravityOverride=true
    gravityScale=0.35
    simulationSpace=RootRelative
    boundBone {boneName=dyn_R_foot_laceTies}
    componentAppliedLinearAccClamp=<4000 4000 4000>
    componentLinearAccScale=<1 1 1>
    componentLinearVelScale=<1 1 1>
    gravityOverride {z=-800}
    physicsBodyDefinitions [{
      boundBone {boneName=dyn_R_foot_laceTies}
      boxExtents=<11 11 11>
      constraintSetup {angularLimitsMax=<0 0 0>, angularLimitsMin=<0 0 -30>, angularTarget=<1 0 0>}
      localJointOffset=<-2 0 0>
     }]
    simSpaceSettings {maxAngularAcceleration=35, maxAngularVelocity=7, simSpaceAngularAlpha=1}
   }
   n15 = "Transform (Modify) Bone - Bone: dyn_neck_02"!(ComponentPose=@n14, bAlphaBoolEnabled:bool=True, Translation="3,0,0", Rotation="0,0,0", Scale="1,1,1") -> (Pose) {
    alphaInputType=Bool
    scaleMode=BMM_Additive
    scaleSpace=BCS_BoneSpace
    translationMode=BMM_Additive
    translationSpace=BCS_BoneSpace
    boneToModify {boneName=dyn_neck_02}
    translation {x=3}
   }
   n16 = "Transform (Modify) Bone - Bone: dyn_neck_01"!(ComponentPose=@n15, bAlphaBoolEnabled:bool=True, Translation="-3,0,0", Rotation="0,0,0", Scale="1,1,1") -> (Pose) {
    alphaInputType=Bool
    scaleMode=BMM_Additive
    scaleSpace=BCS_BoneSpace
    translationMode=BMM_Additive
    translationSpace=BCS_BoneSpace
    boneToModify {boneName=dyn_neck_01}
    translation {x=-3}
   }
   Output Pose(Result=Component To Local(ComponentPose=@n16)) {layerGroup=None}
  ]}
 EventGraph {nodes [
   Event OnNeckHeadHidden() {is=title K2Node_Event}
   > Set IsHiddenHead(IsHiddenHead:bool=true) -> (Output_Get:bool)
   Event OnNeckHeadVisible() {is=title K2Node_Event}
   > Set IsHiddenHead(IsHiddenHead=false)
   Event BlueprintUpdateAnimation()
   > n1 = Sequence!(execute) -> (then_0, then_1, then_2:Exec, then_3:Exec)
   > n2 = PlayerMovement(AnimInstance=Self-Reference(), XDotProductCoefficient=25, YDotProductCoefficient=25, ZDotProductCoefficient=1)
   > Set XMovement(XMovement:float=@n2) -> (Output_Get:float)
   > Set YMovement(YMovement:float=@n2.YMovement) -> (Output_Get:float)
   n3 = PlayerMovement(execute=@n1.then_3, AnimInstance=Self-Reference(), XDotProductCoefficient=1, YDotProductCoefficient=1, ZDotProductCoefficient=1)
   n4 = BreakVector(InVec=MakeVector(X=@n3, Y=@n3.YMovement))
   > Set strafeLeft float(strafeLeft float:double="Absolute (Float)"(A="Clamp (Float)"(Value=@n4, Min=-1, Max=0))) -> (Output_Get:double)
   > Set strafeRight float(strafeRight float:double="Clamp (Float)"(Value=@n4, Min=0, Max=1)) -> (Output_Get:double)
   > Set moveForward(moveForward:double="Clamp (Float)"(Value=@n4.Y, Min=0, Max=1)) -> (Output_Get:double)
   > Set moveBackward(moveBackward:double="Absolute (Float)"(A="Clamp (Float)"(Value=@n4.Y, Min=-1, Max=0))) -> (Output_Get:double)
   n5 = Branch(execute=@n1.then_2, Condition="float < float"(A=Get YMovement() -> (YMovement:float)))
   > n6 = Set YMovement(YMovement=float * float(A=Get YMovement(), B=0.5) {is=title K2Node_PromotableOperator})
   Set JoystickRotator(execute=[@n6.then, @n7.then], JoystickRotator:Rotator=MakeRotator(Pitch=Get XMovement() -> (XMovement:float), Yaw=Get YMovement())) -> (Output_Get:Rotator)
   n7 = Set YMovement(execute=@n5.else, YMovement=Get YMovement())
   n8 = "Absolute (Float)"(A=Get YMovement())
   n9 = "Absolute (Float)"(A=Get XMovement())
   Set movementAlpha(execute=@n1.then_1, movementAlpha:bool=OR Boolean(A="float >= float"(A=@n9, B=0.01), B="float >= float"(A=@n8, B=0.01))) -> (Output_Get:bool)
  ]}
]
variables [
 IsHiddenHead {type=boolean}
 JoystickRotator {type=Rotator}
 moveBackward {type=number}
 moveForward {type=number}
 movementAlpha {type=boolean}
 strafeLeft float {type=number}
 strafeRight float {type=number}
 XMovement {type=number}
 YMovement {type=number}
]
